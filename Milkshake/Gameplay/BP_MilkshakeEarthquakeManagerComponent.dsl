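; Server-authoritative component on an earthquake spline actor that schedules and spawns earthquakes for a named calendar event.
; A repeating timer polls the calendar event; on activation it rolls a random quake limit and starts a spawn-attempt timer.
; Attempts spawn an earthquake actor on the spline only during permitted safe-zone phases with no registered quake in the world, otherwise retrying.
; Each quake's destruction starts a random cooldown until the limit is reached.
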(event EventBeginPlay
  ;[0 0.3073 0.04605] BeginPlay: Start the timer to check and update the calendar event state.
  (|CallParentFunction)
  (Networking|SwitchHasAuthority
    (:Authority
      (bind _returnvalue (Utilities|Time|SetTimerbyEvent (EventDispatchers|CreateEvent :!event UpdateCalendarEventActiveStatus) (ScalableFloat|GetValueatLevel (Variables|Default|GetCalendarEventStatusUpdateInterval)) true false 0.5 0.2))
      (Variables|Default|SetUpdateCalendarEventStatusTimerHandle _returnvalue)
      (Fortnite|FortShippingLog self (Utilities|String|Append "BeginPlay: Successfully setup earthquake component for event " (Variables|Default|GetEarthquakeEventName))))))

(event Custom|BeginEarthquakeEvent
  ;[0 0.1927 0.1787] BeginEarthquakeEvent: Setup the earthquake properties such as max number of earthquakes to spawn.
  ; Start the earthquake spawn timer.
  (Networking|SwitchHasAuthority
    (:Authority
      (Fortnite|FortShippingLog self (Utilities|String|Append "BeginEarthquakeEvent: Starting earthquake event for " (Variables|Default|GetEarthquakeEventName)))
      (CallFunction|SetupEarthquakeComponentProperties)
      (CallFunction|StartEarthquakeAttemptSpawnTimer))))

(event Custom|EndEarthquakeEvent
  ;[0 0.03868 0.08333] EndEarthquakeEvent: Stop the spawn earthquake timer if it was running.
  ; Log the status of the current earthquake.
  (Networking|SwitchHasAuthority
    (:Authority
      (Fortnite|FortShippingLog self (Utilities|String|Append "EndEarthquakeEvent: Stopping earthquake event for " (Variables|Default|GetEarthquakeEventName)))
      (Utilities|Time|ClearandInvalidateTimerbyHandle (Variables|Default|GetEarthquakeSpawnCooldownTimerHandle))
      (Utilities|Time|ClearandInvalidateTimerbyHandle (Variables|Default|GetAttemptSpawnEarthquakeTimerHandle))
      (bind _currentearthquake (Variables|Default|GetCurrentEarthquake)
        (:then
          (Fortnite|FortShippingLog self (Utilities|String|Append "EndEarthquakeEvent: Event " (Variables|Default|GetEarthquakeEventName) " has ended but CurrentEarthquake is still active as " (Utilities|GetDisplayName _currentearthquake))))
        (:else
          (Fortnite|FortShippingLog self (Utilities|String|Append "EndEarthquakeEvent: Event " (Variables|Default|GetEarthquakeEventName) " has ended and there is no current earthquake occuring;")))))))

(event Custom|SpawnEarthquake
  (bind _splineactor (CallFunction|GetOwningEarthquakeSplineActor))
  ;[0.174 0 0.1875] SpawnEarthquake: Called after the timer to spawn an earthquake is complete;
  ; Spawns the earthquake actor in the world that is responsible for building damage, screenshake etc.
  (Networking|SwitchHasAuthority
    (:Authority
      (Utilities|Time|ClearandInvalidateTimerbyHandle (Variables|Default|GetAttemptSpawnEarthquakeTimerHandle))
      (bind _bspawn (CallFunction|CanSpawnEarthquake))
      (if _bspawn
        (bind _info (CallFunction|GetEarthquakeInfo))
        (Utilities|IsValid _splineactor
          (:Is%20Valid
            (bind _returnvalue (Game|SpawnActorfromClass "/Milkshake/Gameplay/BP_Milkshake_Earthquake.BP_Milkshake_Earthquake_C" (Math|Transform|MakeTransform (Class|BPMilkshakeSplineEarthquake|GetSplineLocationatAlpha _splineactor (Math|Vector2D|BreakVector2D (.EarthquakeSplineMoveNormalizedLength_110_2E9091BE45B8E7FF70C161AE980C69A9 (Utilities|Struct|BreakMilkshakeEarthquakeInfoStruct _info))))) "AlwaysSpawn" "OverrideRootScale" 0 _splineactor _info self 0 0 0))
            (bind _output_get (Variables|Default|SetCurrentEarthquake _returnvalue))
            (Utilities|IsValid _output_get
              (:Is%20Valid
                (Game|BindEventtoOnDestroyed _output_get (EventDispatchers|CreateEvent :!event Custom|HandleEarthquakeActorDestroyed))
                (bind _result (Math|Integer|IncrementInt (Variables|Default|GetNumOfEarthquakesSpawned))
                  (:%20%20
                    (Fortnite|FortShippingLog self (Utilities|String|Append "SpawnEarthquake: Successfully spawned earthquake " (Utilities|GetDisplayName _output_get) " for event " (Variables|Default|GetEarthquakeEventName) ". NumOfEarthquakesSpawnd: " (Utilities|String|ToString(Integer) _result))))))
              (:Is%20Not%20Valid
                ;;[1 0 0] Warning: Failed to spawn earthquake cause the spawn actor function failed.
                (Fortnite|FortShippingLog self (Utilities|String|Append "SpawnEarthquake: Failed to spawn earthquake since SpawnActor function failed for event " (Variables|Default|GetEarthquakeEventName)) true)
                (CallFunction|StartEarthquakeAttemptSpawnTimer))))
          (:Is%20Not%20Valid
            ;;[1 0 0] Warning: Failed to spawn earthquake cause owning spline actor is null.
            (Fortnite|FortShippingLog self (Utilities|String|Append "SpawnEarthquake: Failed to spawn earthquake since the owning spline earthquake actor is null for event " (Variables|Default|GetEarthquakeEventName)) true)))
        (else
          (Fortnite|FortShippingLog self (Utilities|String|Append "SpawnEarthquake: Unable to spawn earthquake this turn for event " (Variables|Default|GetEarthquakeEventName) ". Restarting spawn timer."))
          (CallFunction|StartEarthquakeAttemptSpawnTimer))))))

(event Custom|HandleEarthquakeActorDestroyed
  (params (DestroyedActor :type Actor%20Object%20Reference))
  (bind _currentearthquake (Variables|Default|GetCurrentEarthquake))
  ;[0.1354 0.0886 0] HandleEarthquakeActorDestroyed: Called when the current earthquake actor has finished its course.
  (Networking|SwitchHasAuthority
    (:Authority
      (Utilities|IsValid DestroyedActor
        (:Is%20Valid
          (if (== DestroyedActor _currentearthquake)
            (Variables|Default|SetCurrentEarthquake 0)
            (Fortnite|FortShippingLog self (Utilities|String|Append "HandleEarthquakeActorDestroyed: Current earthquake actor " (Utilities|GetDisplayName _currentearthquake) " destroyed for event " (Variables|Default|GetEarthquakeEventName)))
            (Game|UnbindEventfromOnDestroyed DestroyedActor (EventDispatchers|CreateEvent :!event Custom|HandleEarthquakeActorDestroyed))
            (bind _bactive (CallFunction|IsOwningCalendarEventActive))
            (if _bactive
              ;;[0.2188 0 0.1551] Spawn the next earthquake if MAX num of quakes has NOT been hit.
              (if (or (< (Variables|Default|GetMaxNumOfEarthquakes) 0) (< (Variables|Default|GetNumOfEarthquakesSpawned) (Variables|Default|GetMaxNumOfEarthquakes)))
                (CallFunction|StartEarthquakeSpawnCooldownTimer)
                (else
                  (Fortnite|FortShippingLog self (Utilities|String|Append "HandleEarthquakeActorDestroyed: Max num of earthquakes hit for event " (Variables|Default|GetEarthquakeEventName) ". Not spawning anymore earthquakes. MaxEarthquakeCount: " (Utilities|String|ToString(Integer) (Variables|Default|GetMaxNumOfEarthquakes))))))
              (else
                (Fortnite|FortShippingLog self (Utilities|String|Append "HandleEarthquakeActorDestroyed: Earthquake Event " (Variables|Default|GetEarthquakeEventName) " is no longer active and thus not spawning any more earthquakes."))))
            (else
              ;;[1 0 0] Warning: Callback from unexpected actor.
              (Fortnite|FortShippingLog self (Utilities|String|Append "HandleEarthquakeActorDestroyed: Unexpected callback from actor " (Utilities|GetDisplayName DestroyedActor) " when CurrentEarthquake value is " (Utilities|GetDisplayName (Variables|Default|GetCurrentEarthquake)) " for event name " (Variables|Default|GetEarthquakeEventName)) true))))))))

(event EventEndPlay
  (params (EndPlayReason :type EEndPlayReason%20Enum :default "Destroyed"))
  ;[1 0.5811 0.03215] EndPlay: Destroy the current earthquake actor if it is still valid.
  (|CallParentFunction EndPlayReason)
  (Networking|SwitchHasAuthority
    (:Authority
      (Utilities|Time|ClearandInvalidateTimerbyHandle (Variables|Default|GetAttemptSpawnEarthquakeTimerHandle))
      (Utilities|Time|ClearandInvalidateTimerbyHandle (Variables|Default|GetUpdateCalendarEventStatusTimerHandle))
      (Utilities|Time|ClearandInvalidateTimerbyHandle (Variables|Default|GetEarthquakeSpawnCooldownTimerHandle))
      (Fortnite|FortShippingLog self (Utilities|String|Append "EndPlay: Destroying milkshake earthquake manager component for event name " (Variables|Default|GetEarthquakeEventName)))
      (bind _currentearthquake (Variables|Default|GetCurrentEarthquake)
        (:then
          (Actor|DestroyActor _currentearthquake))
        (:else)))))

(fn CanSpawnEarthquake
  (bind _returnvalue (BattleRoyaleGamePhaseLogic|GetBattleRoyaleGamePhaseLogicComponent))
  (bind _returnvalue_2 (BattleRoyaleGamePhaseLogic|GetSafeZonePhaseIndex _returnvalue))
  (bind _returnvalue_3 (WorldSubsystems|GetFortGlobalActorRegistrySubsystem))
  ;[0.1525 0.1615 0] CanSpawnEarthquake: Returns if the earthquake can be successfully spawned.
  ; Prints warning logs for unexpected conditions.
  (Utilities|FlowControl|Sequence
    (:then_0
      (Networking|SwitchHasAuthority
        (:Authority
          (bind _bactive (CallFunction|IsOwningCalendarEventActive))
          (if _bactive
            ;;[0.07774 0.0107 0.07812] Can only spawn if there are NO other active earthquakes in the world.
            (Utilities|IsValid _returnvalue_3
              (:Is%20Valid
                (bind _outactors (Class|FortGlobalActorRegistrySubsystem|GetRegisteredActors _returnvalue_3 (Variables|Default|GetEarthquakeActorRegisterTag)))
                (if (<= (Utilities|Array|Length _outactors) 0)
                  (Variables|Default|GetCurrentEarthquake
                    (:then
                      ;;[1 0 0] Warning: CurrentEarthquake is already valid.
                      (Fortnite|FortShippingLog self (Utilities|String|Append "CanSpawnEarthquake: Cannot spawn earthquake for event " (Variables|Default|GetEarthquakeEventName) " since CurrentEarthquake is already valid as " (Utilities|GetDisplayName (Variables|Default|GetCurrentEarthquake))) true))
                    (:else
                      (bind _info (CallFunction|GetEarthquakeInfo))
                      (Utilities|IsValid _returnvalue
                        (:Is%20Valid
                          (if (or (Utilities|Enum|Equal(Enum) (BattleRoyaleGamePhaseLogic|GetGamePhase _returnvalue) "SafeZones") (Utilities|Enum|Equal(Enum) (BattleRoyaleGamePhaseLogic|GetGamePhase _returnvalue) "EndGame"))
                            (if (Utilities|Array|IsValidIndex (.SafeZonePhase_106_F9A528FA4C9CFAF4C009FD820928814E (Utilities|Struct|BreakMilkshakeEarthquakeInfoStruct _info)) _returnvalue_2)
                              (if (Utilities|Array|Get(acopy) (.SafeZonePhase_106_F9A528FA4C9CFAF4C009FD820928814E (Utilities|Struct|BreakMilkshakeEarthquakeInfoStruct _info)) _returnvalue_2)
                                (return true)
                                (else
                                  _))
                              (else
                                _))
                            (else
                              _)))
                        (:Is%20Not%20Valid
                          ;;[1 0 0] Warning: GamePhaseLogicComponent is invalid.
                          (Fortnite|FortShippingLog self (Utilities|String|Append "CanSpawnEarthquake: Cannot spawn earthquake for event " (Variables|Default|GetEarthquakeEventName) " since GamePhaseLogicComponent is NOT valid.") true)))))
                  (else
                    (Fortnite|FortShippingLog self (Utilities|String|Append "CanSpawnEarthquake: Cannot spawn earthquake for event " (Variables|Default|GetEarthquakeEventName) " since there's already an earthquake actor in the world.")))))
              (:Is%20Not%20Valid
                ;;[1 0 0] Warning: FortGlobalActorRegistrySubsystem is null.
                (Fortnite|FortShippingLog self (Utilities|String|Append "CanSpawnEarthquake: Cannot spawn earthquake for event " (Variables|Default|GetEarthquakeEventName) " since FortGlobalActorRegistrySubsystem is null.") true)))
            (else
              ;;[1 0 0] Warning: Calendar Event is NOT active.
              (Fortnite|FortShippingLog self (Utilities|String|Append "CanSpawnEarthquake: Cannot spawn earthquake for event " (Variables|Default|GetEarthquakeEventName) " since the calendar event is NOT active.") true))))))
    (:then_1
      (return false))))

(fn GetEarthquakeInfo
  (bind _splineactor (CallFunction|GetOwningEarthquakeSplineActor))
  (bind _earthquakeeventname (Variables|Default|GetEarthquakeEventName))
  (bind (_info _bfound) (Class|BPMilkshakeSplineEarthquake|GetEarthquakeInfoforEvent _splineactor _earthquakeeventname))
  ;[0.04203 0.07828 0.1562] Returns the earthquake info gathered from the parent spline actor.
  (Utilities|FlowControl|Sequence
    (:then_0
      (Utilities|IsValid _splineactor
        (:Is%20Valid
          (if _bfound
            (return _info)
            (else
              ;;[1 0 0] Warning: EarthquakeInfo is invalid.
              (Fortnite|FortShippingLog self (Utilities|String|Append "GetEarthquakeInfo: EarthquakeInfo is invalid for EventName " (Variables|Default|GetEarthquakeEventName)) true))))
        (:Is%20Not%20Valid
          ;;[1 0 0] Warning: Owning Earthquake Spline Actor is invalid.
          (Fortnite|FortShippingLog self (Utilities|String|Append "GetEarthquakeInfo: Owning Earthquake Spline Actor is invalid for EventName " (Variables|Default|GetEarthquakeEventName)) true))))
    (:then_1
      (return 0))))

(fn GetEventName
  ;[0.0152 0.6218 1] Returns the event name.
  (return (Variables|Default|GetEarthquakeEventName)))

(fn GetOwningEarthquakeSplineActor
  (bind _returnvalue (Components|GetOwner))
  ;[0 0.07082 1] Returns the owning spline actor.
  (bind _asbp_milkshake_spline_earthquake (Utilities|Casting|CastToBP_Milkshake_Spline_Earthquake _returnvalue)
    (:then
      (return _asbp_milkshake_spline_earthquake))
    (:CastFailed
      ;;[1 0 0] Warning: Owning actor is of the wrong class.
      (Fortnite|FortShippingLog self (Utilities|String|Append "GetOwningEarthquakeSplineActor: The owning actor is NOT of class BP_Milkshake_Spline_Earthquake. Owning actor: " (Utilities|GetDisplayName _returnvalue)) true)
      (return 0))))

(fn HandleEarthquakeSpawnCooldownTimerComplete
  ;[0.1963 0.003612 0.2552] HandleEarthquakeSpawnCooldownTimerComplete: Called when the earthquake cooldown timer has completed.
  (Networking|SwitchHasAuthority
    (:Authority
      (Fortnite|FortShippingLog self (Utilities|String|Append "HandleEarthquakeSpawnCooldownTimerComplete: Invoked for event " (Variables|Default|GetEarthquakeEventName)))
      (CallFunction|StartEarthquakeAttemptSpawnTimer))))

(fn IsOwningCalendarEventActive
  (bind _splineactor (CallFunction|GetOwningEarthquakeSplineActor))
  ;[0.01015 0.03082 0.1198] Returns if the owning calendar event is active.
  (Utilities|FlowControl|Sequence
    (:then_0
      (Utilities|IsValid _splineactor
        (:Is%20Valid
          (bind _bactive (Class|BPMilkshakeSplineEarthquake|IsCalendarEventActive _splineactor (Variables|Default|GetEarthquakeEventName)))
          (return _bactive))))
    (:then_1
      (return false))))

(fn SetupEarthquakeComponentProperties
  ;[0.09896 0 0.08624] SetupEarthquakeComponentProperties: Calculate the maximum number of earthquakes that this component should spawn for the session/match.
  (bind _info (CallFunction|GetEarthquakeInfo))
  (bind _output_get (Variables|Default|SetMaxNumOfEarthquakes (Math|Random|RandomIntegerinRange (Math|Float|Truncate (Math|Vector2D|BreakVector2D (.NumberOfEarthquakes_93_299C58AD466ABA23C371AEB841D710A5 (Utilities|Struct|BreakMilkshakeEarthquakeInfoStruct _info)))) (Math|Float|Truncate (.Y (Math|Vector2D|BreakVector2D (.NumberOfEarthquakes_93_299C58AD466ABA23C371AEB841D710A5 (Utilities|Struct|BreakMilkshakeEarthquakeInfoStruct _info))))))))
  (Fortnite|FortShippingLog self (Utilities|String|Append "SetupEarthquakeComponentProperties: Setting MaxNumOfEarthquakes for event " (Variables|Default|GetEarthquakeEventName) " to " (Utilities|String|ToString(Integer) _output_get))))

(fn StartEarthquakeAttemptSpawnTimer
  (bind _earthquakeattemptspawninterval (Variables|Default|GetEarthquakeAttemptSpawnInterval))
  (bind _returnvalue (ScalableFloat|GetValueatLevel _earthquakeattemptspawninterval))
  ;[0 0.1159 0.2396] StartEarthquakeAttemptSpawnTimer: Starts a timer with the desired value after which the earthquake is attempted to be spawned.
  (Networking|SwitchHasAuthority
    (:Authority
      (if (not (Utilities|Time|IsTimerActivebyHandle (Variables|Default|GetAttemptSpawnEarthquakeTimerHandle)))
        (bind _returnvalue_2 (Utilities|Time|SetTimerbyEvent (EventDispatchers|CreateEvent :!event SpawnEarthquake) _returnvalue))
        (Variables|Default|SetAttemptSpawnEarthquakeTimerHandle _returnvalue_2)
        (Fortnite|FortShippingLog self (Utilities|String|Append "StartEarthquakeAttemptSpawnTimer: Setting a spawn timer for event " (Variables|Default|GetEarthquakeEventName) " with a time value of  " (Utilities|String|ToString(Float) _returnvalue)))
        (else
          ;;[1 0 0] Warning: Timer is already active.
          (Fortnite|FortShippingLog self (Utilities|String|Append "StartEarthquakeAttemptSpawnTimer: Looking to set a spawn timer for event " (Variables|Default|GetEarthquakeEventName) " when a timer is already active for the same.") true))))))

(fn StartEarthquakeSpawnCooldownTimer
  ;[0.01018 0.08725 0.1406] Start the cooldown between two individual earthquakes.
  (Networking|SwitchHasAuthority
    (:Authority
      (bind _info (CallFunction|GetEarthquakeInfo))
      (bind _output_get (LocalVariables|SetTimerDuration (Math|Random|RandomFloatinRange (Math|Vector2D|BreakVector2D (.DelayBetweenEarthquakes_91_C25849394C195BDA2B398D9BEC9E0B9D (Utilities|Struct|BreakMilkshakeEarthquakeInfoStruct _info))) (.Y (Math|Vector2D|BreakVector2D (.DelayBetweenEarthquakes_91_C25849394C195BDA2B398D9BEC9E0B9D (Utilities|Struct|BreakMilkshakeEarthquakeInfoStruct _info)))))))
      (bind _returnvalue (Utilities|Time|SetTimerbyEvent (EventDispatchers|CreateEvent :!event HandleEarthquakeSpawnCooldownTimerComplete) _output_get))
      (Variables|Default|SetEarthquakeSpawnCooldownTimerHandle _returnvalue)
      (Fortnite|FortShippingLog self (Utilities|String|Append "StartEarthquakeSpawnCooldownTimer: Setting a cooldown timer for event " (Variables|Default|GetEarthquakeEventName) " with a time value of  " (Utilities|String|ToString(Float) _output_get))))))

(fn UpdateCalendarEventActiveStatus
  ;[0.004168 0.0311 0.04688] Checks the status of the calendar event and calls the respective event if the status has changed.
  (Networking|SwitchHasAuthority
    (:Authority
      (bind _bactive (CallFunction|IsOwningCalendarEventActive))
      (if (!= _bactive (Variables|Default|GetbCalendarEventActive))
        (bind _output_get (Variables|Default|SetbCalendarEventActive _bactive))
        (if _output_get
          (CallFunction|BeginEarthquakeEvent)
          (else
            (CallFunction|EndEarthquakeEvent)))))))
