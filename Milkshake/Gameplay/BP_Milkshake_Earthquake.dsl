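; Server-driven Milkshake earthquake actor that moves along a spline and damages buildings, pushes vehicles and notifies registered react actors within its radius.
; A repeating async overlap trace maintains the overlapped-actor list, feeding physics force registration and building damage tasks.
; Damage scales with epicenter distance, spreads to neighbors of weakened pieces, and runs in delayed batches to spare the server.
; Clients run Niagara FX, audio and a radial force actor for foliage.
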(event Custom|AddEarthquakeReactActorListeners
  (bind _returnvalue (WorldSubsystems|GetFortGlobalActorRegistrySubsystem))
  (bind _earthquakereactactortag (Variables|ReactActors|GetEarthquakeReactActorTag))
  ;[0.09621 1 0] Execute from Begin Play: Add listener for registering actors and call OnActorRegistered on all currently registered actors.
  (Networking|SwitchHasAuthority
    (:Authority
      (Utilities|IsValid _returnvalue
        (:Is%20Valid
          (bind _returnvalue_2 (Class|FortGlobalActorRegistrySubsystem|GetActorRegistryDataforGameplayTag _returnvalue _earthquakereactactortag))
          (Utilities|IsValid _returnvalue_2
            (:Is%20Valid
              (EventDispatchers|BindEventtoOnActorRegistered _returnvalue_2 (EventDispatchers|CreateEvent :!event Custom|HandleEarthquakeReactActorRegistered))
              (EventDispatchers|BindEventtoOnActorUnregistered _returnvalue_2 (EventDispatchers|CreateEvent :!event Custom|HandleEarthquakeReactActorUnregistered))
              ;;[0 1 0.9234] Call OnActorRegistered for all currently registered actors
              (for _array_element (Class|ActorRegisterDataGameplayTagBased|GetRegisteredActors _returnvalue_2)
                (Utilities|IsValid _array_element
                  (:Is%20Valid
                    (CallFunction|HandleEarthquakeReactActorRegistered _earthquakereactactortag _array_element)))))))
        (:Is%20Not%20Valid
          (Fortnite|FortShippingLog self "FortHandleRegisteredActor_Base::BeginPlay. Failed to get GlobalActorRegistry." true))))))

(event Custom|HandleEarthquakeReactActorRegistered
  (params (GameplayTag :type Gameplay%20Tag%20Structure%20%28by%20ref%29) (RegisteredActor :type Actor%20Object%20Reference))
  ;[0.08447 0.06935 0.3333] HandleEarthquakeReactActorRegistered: Add the newly registering actor to the list of registered actors.
  (Networking|SwitchHasAuthority
    (:Authority
      (Utilities|IsValid RegisteredActor
        (:Is%20Valid
          (bind _asbp_milkshake_react_parent (Utilities|Casting|CastToBP_Milkshake_React_Parent RegisteredActor)
            (:then
              (Utilities|Array|Add (Variables|ReactActors|GetRegisteredEarthquakeReactActors) _asbp_milkshake_react_parent)
              (Fortnite|FortShippingLog self (Utilities|String|Append "HandleEarthquakeReactActorRegistered: Registering actor: " (Utilities|GetDisplayName _asbp_milkshake_react_parent))))
            (:CastFailed
              ;;[1 0 0] Warning: Input actor is not of expected type.
              (Fortnite|FortShippingLog self (Utilities|String|Append "HandleEarthquakeReactActorRegistered: Input actor is not of Milkshake_React_Parent type: " (Utilities|GetDisplayName RegisteredActor)) true))))))))

(event Custom|HandleEarthquakeReactActorUnregistered
  (params (GameplayTag :type Gameplay%20Tag%20Structure%20%28by%20ref%29) (UnregisteredActor :type Actor%20Object%20Reference))
  ;[0.08447 0.06935 0.3333] HandleEarthquakeReactActorUnregistered: Remove the unregistering actor from the react actors list.
  (Networking|SwitchHasAuthority
    (:Authority
      (Utilities|IsValid UnregisteredActor
        (:Is%20Valid
          (bind _asbp_milkshake_react_parent (Utilities|Casting|CastToBP_Milkshake_React_Parent UnregisteredActor))
          (Utilities|Array|RemoveItem (Variables|ReactActors|GetRegisteredEarthquakeReactActors) _asbp_milkshake_react_parent)
          (Fortnite|FortShippingLog self (Utilities|String|Append "HandleEarthquakeReactActorRegistered: Unregistering actor: " (Utilities|GetDisplayName _asbp_milkshake_react_parent))))))))

(event Custom|RemoveEarthquakeReactActorListeners
  (bind _returnvalue (WorldSubsystems|GetFortGlobalActorRegistrySubsystem))
  ;[1 0.5885 0] Execute on EndPlay: Remove earthquake react actor listeners.
  (Networking|SwitchHasAuthority
    (:Authority
      (Utilities|IsValid _returnvalue
        (:Is%20Valid
          (bind _returnvalue_2 (Class|FortGlobalActorRegistrySubsystem|GetActorRegistryDataforGameplayTag _returnvalue (Variables|ReactActors|GetEarthquakeReactActorTag)))
          (Utilities|IsValid _returnvalue_2
            (:Is%20Valid
              (EventDispatchers|UnbindEventfromOnActorRegistered _returnvalue_2 (EventDispatchers|CreateEvent :!event Custom|HandleEarthquakeReactActorRegistered))
              (Class|ActorRegisterDataGameplayTagBased|HandleListenerRemoved _returnvalue_2))))))))

(event Custom|UpdateRegisteredReactActorsOverlapStatus
  ;[0 0.07652 0.1719] UpdateRegisteredReactActorsOverlapStatus: Update the overlapping status of the registered react actors using the 'UpdateEarthquakeOverlappingStatus' function.
  (Networking|SwitchHasAuthority
    (:Authority
      (for _array_element (Variables|ReactActors|GetRegisteredEarthquakeReactActors)
        (Utilities|IsValid _array_element
          (:Is%20Valid
            (bind _bwithinradius (Overlap|IsActorWithinEarthquakeRadius _array_element))
            (Class|BPMilkshakeReactParent|UpdateEarthquakeOverlappingStatus _array_element _bwithinradius self)))))))

(event EventBeginPlay
  ;[0 0.3594 0] BeginPlay: On the server, setup unique tag actor names to be used for this earthquake actor.
  (|CallParentFunction)
  (Networking|SwitchHasAuthority
    (:Authority
      (General|SetupUniqueActorTagNameVariables)
      (General|SetReplicatedEarthquakeInfo)
      (CallFunction|AddEarthquakeReactActorListeners)
      (Networking|FlushNetDormancy))
    (:Remote
      (Actor|Tick|SetActorTickEnabled false)
      (General|SetRegistertoPlayerControllerComponents true)
      (bind _returnvalue (Audio|SpawnSoundatLocation (Variables|Audio|GetEarthquake%20Sound) (Transformation|GetWorldLocation (Variables|Default|GetDefaultSceneRoot))))
      ;;[1 0 0.6885] Setup actor FX.
      (Variables|Audio|SetEarthquake%20Audio%20Component _returnvalue)
      (Utilities|FlowControl|Delay (Variables|FX|GetDelayBeforeQuakeFXSpawn))
      (bind _returnvalue_2 (Niagara|SpawnSystematLocation (Variables|FX|GetQuakeFXSystem) (Transformation|GetActorLocation)))
      (Variables|Default|SetQuakeFX _returnvalue_2)
      (Utilities|Time|SetTimerbyEvent (EventDispatchers|CreateEvent :!event UpdateQuakeFXParams) (Variables|FX|GetQuakeFXParamUpdateInterval) true))))

(event Custom|BeginEarthquake
  (bind _serverearthquakeinfostruct (Variables|General|GetServerEarthquakeInfoStruct))
  (bind (_enabled_104_9e9bd6984687100ae0fd3cb84b7e6a76 _calendareventratio_116_db14dea24c94389f1ded8fad08f78569 _earthquakesplinemovenormalizedlength_110_2e9091be45b8e7ff70c161ae980c69a9) (Utilities|Struct|BreakMilkshakeEarthquakeInfoStruct _serverearthquakeinfostruct))
  (bind (_earthquakestartgametime_111_1f928ca54731f603a460a7b5e81d3536 _earthquakeendgametime_109_a139a81e4ea551f5bbd8c5ab7705588b) (Utilities|Struct|BreakMilkshakeEarthquakeRepInfoStruct (Variables|General|GetEarthquakeRepInfo)))
  (bind (_x _y) (Math|Vector2D|BreakVector2D _earthquakesplinemovenormalizedlength_110_2e9091be45b8e7ff70c161ae980c69a9))
  (bind _shouldapplybuildingdamage (Variables|Default|GetShouldApplyBuildingDamage))
  (bind _returnvalue (ScalableFloat|GetValueasBool _shouldapplybuildingdamage))
  ;[0.1364 0.3802 0.1198] BeginEarthquake: On the server, apply the milkshake component to the player controllers. The milkshake component is responsible for handling the screenshake, audio feedback, movement speed reduction, etc.
  ; On the server, handle the movement of the earthquake along the spline.
  ; On the client, handle the radial force actor for foliage and trees.
  (if (not (Variables|General|GetbEarthquakeActive))
    (Variables|General|SetbEarthquakeActive true)
    (Networking|SwitchHasAuthority
      (:Authority
        (Utilities|FlowControl|Sequence
          (:then_0
            ;;[0 0.755 1] On the server, start a looping timer
            ;; that updates the overlap status
            ;; of the registered react actors.
            (Utilities|Time|SetTimerbyEvent (EventDispatchers|CreateEvent :!event Custom|UpdateRegisteredReactActorsOverlapStatus) (Variables|ReactActors|GetUpdateEarthquakeReactActorsOverlapStatusInterval) true false 0.1 0.1)
            ;;[0.001022 0 0.2812] On the server, move the earthquake along the spline based on the set up properties.
            (Components|Timeline|SetPlayRate (Variables|BP_Milkshake_Earthquake|GetMovementTimeline) (Math|Float|SafeDivide 1.0 (- _earthquakeendgametime_109_a139a81e4ea551f5bbd8c5ab7705588b _earthquakestartgametime_111_1f928ca54731f603a460a7b5e81d3536)))
            (if (!= _x _y)
              (bind (_direction _alpha) (Timeline|MovementTimeline Play)
                (:Update
                  (bind _splineactor (Variables|Movement|GetSplineActor)
                    (:then
                      (Transformation|SetActorLocation (Class|BPMilkshakeSplineEarthquake|GetSplineLocationatAlpha _splineactor (Math|Float|Lerp _x _y _alpha))))
                    (:else))))))
          (:then_1
            (if _returnvalue
              ;;[0 0.2917 0.1578] On the server, start running async overlap trace.
              ;; This trace manages the AsyncOverlappedActors array which at any time holds the actors in the world that are currently being overlapped by the earthquake. Note: This array can be large.
              (Overlap|RunAsyncOverlapTrace)))
          (:then_2
            (if _returnvalue
              ;;[0 0.1024 0.2031] On the server, start looking for building actors to damage.
              ;; These values are cached as floats since they are going to be used multiple times on the same frame. Optimization pass for not having to read the curve multiplers on the same frame for the same values.
              (Variables|BuildingDamage|SetBuildingDamageSpreadThresholdHealthPercent (ScalableFloat|GetValueatLevel (Variables|BuildingDamage|GetBuildingDamageSpreadThresholdHealthPercent_SF)))
              (CallFunction|StartDamageBuildingActors)))))
      (:Remote
        ;;[0.3125 0.1579 0] On the client, spawn the radial force actor responsible for handling the tree/foilage (material and effects) movements in the area.
        (FX|SpawnRadialForceActor)
        ;;[0.01243 0.1938 0.2083] Spawn the radius sphere IF that cheat is enabled.
        (Cheat|CheatSpawnRadiusSphere)))))

(event EventTick
  (params (DeltaSeconds :type Float%20%28single-precision%29))
  ;[0 0.1776 0.4531] Tick: On the server, handle the force impulses that get applied to the physics actors in the earthquake.
  (|CallParentFunction DeltaSeconds)
  (Networking|SwitchHasAuthority
    (:Authority
      (General|TickDebugEarthquake)
      (if (ScalableFloat|GetValueasBool (Variables|Default|GetShouldApplyForceToPhysicsObjects))
        (CallFunction|TickEarthquakePhysicsActorForces DeltaSeconds)
        (else
          _))
      (General|TickLifetime))
    (:Remote
      (Actor|Tick|SetActorTickEnabled false))))

(event Custom|HandleEarthquakeDurationComplete
  ;[0.0122 0.2448 0.09797] HandleEarthquakeDurationComplete: Destroy the earthquake actor once the duration is complete.
  (Networking|SwitchHasAuthority
    (:Authority
      (Actor|DestroyActor))))

(event EventEndPlay
  (params (EndPlayReason :type EEndPlayReason%20Enum :default "Destroyed"))
  (bind _asyncoverlappedactors (Variables|AsyncOverlapActors|GetAsyncOverlappedActors))
  (bind _buildingactorstodamageforcurrenttask (Variables|BuildingDamage|GetBuildingActorsToDamageForCurrentTask))
  (bind _buildingactorstodamageforcurrenttaskloop (Variables|BuildingDamage|GetBuildingActorsToDamageForCurrentTaskLoop))
  ;[1 0.54 0] EndPlay: On the sever, remove the milkshake component from the controllers and invalidate timer handles.
  ; On the client, remove the radial force actor.
  (|CallParentFunction EndPlayReason)
  (Networking|SwitchHasAuthority
    (:Authority
      (CallFunction|RemoveEarthquakeReactActorListeners)
      (Utilities|FlowControl|Sequence
        (:then_0
          (for _array_element (Utilities|Array|MakeArray (Variables|Overlap|GetAsyncOverlapTraceTimerHandle) (Variables|BuildingDamage|GetBuildingActorsDamageTimerHandle))
            (Utilities|Time|ClearandInvalidateTimerbyHandle _array_element)))
        (:then_1
          (General|RemoveActorTagfromActors _buildingactorstodamageforcurrenttask (Variables|General|GetEarthquakeDamageBuildingActorTag))
          (Utilities|Array|Clear _buildingactorstodamageforcurrenttask)
          (General|RemoveActorTagfromActors _buildingactorstodamageforcurrenttaskloop (Variables|General|GetEarthquakeDamageBuildingActorTag))
          (Utilities|Array|Clear _buildingactorstodamageforcurrenttaskloop))
        (:then_2
          (for _array_element_2 _asyncoverlappedactors
            (Utilities|IsValid _array_element_2
              (:Is%20Valid
                (Game|UnbindEventfromOnDestroyed _array_element_2 (EventDispatchers|CreateEvent :!event Custom|HandleOverlappedActorDestroyed))
                (General|RemoveActorTagfromActors (Utilities|Array|MakeArray _array_element_2) (Variables|General|GetEarthquakeIdentificationName))
                (CallFunction|HandleActorEndAsyncOverlap _array_element_2))))
          (Utilities|Array|Clear _asyncoverlappedactors))
        (:then_3
          (for _array_element_3 (Variables|ReactActors|GetRegisteredEarthquakeReactActors)
            (Utilities|IsValid _array_element_3
              (:Is%20Valid
                (Class|BPMilkshakeReactParent|UpdateEarthquakeOverlappingStatus _array_element_3 false self)))))))
    (:Remote
      (Utilities|FlowControl|Sequence
        (:then_0
          (bind _radialforce (Variables|RadialForce|GetRadialForce)
            (:then
              (Actor|DestroyActor _radialforce))
            (:else)))
        (:then_1
          (bind _quakefx (Variables|Default|GetQuakeFX)
            (:then
              (Components|Activation|Deactivate _quakefx))
            (:else)))
        (:then_2
          (bind _cheat_radiussphere (Variables|Cheat|GetCheat_RadiusSphere)
            (:then
              (Actor|DestroyActor _cheat_radiussphere))
            (:else)))
        (:then_3
          (bind _earthquake_audio_component (Variables|Audio|GetEarthquake%20Audio%20Component)
            (:then
              (Audio|Components|Audio|FadeOut _earthquake_audio_component (Variables|Audio|GetEarthquake%20Fade%20Out%20Duration) 0.0 (Variables|Audio|GetEarthquake%20Fade%20Curve) :!class /Script/Engine.AudioComponent))
            (:else)))
        (:then_4
          (General|SetRegistertoPlayerControllerComponents false))))))

(event Custom|StartDamageBuildingActors
  ;[0 0.05547 0.09896] StartDamageBuildingActors: Starting the damage behavior towards buildings.
  ; Damage to building actors is done in the following steps.
  ; Step 1: Find the building actors overlapped by the earthquake.
  ; Step 2: Damage a set number of these found actors periodically until all building actors found by the overlap are damaged.
  ; Step 3: Wait for the desired cooldown duration before repeating from step 1.
  ; For nomenclature, we are calling each set of steps a 'task'.
  (Networking|SwitchHasAuthority
    (:Authority
      (Fortnite|FortShippingLog self "StartDamageBuildingActors: Starting the damage behavior towards buildings.")
      (CallFunction|StartNextBuildingActorsDamageTask))))

(event Custom|StartNextBuildingActorsDamageTask
  (bind _damagenextbuildingsetinterval (Variables|BuildingDamage|GetDamageNextBuildingSetInterval))
  (bind _returnvalue (ScalableFloat|GetValueatLevel _damagenextbuildingsetinterval))
  (bind _buildingactorstodamageforcurrenttask (Variables|BuildingDamage|GetBuildingActorsToDamageForCurrentTask))
  (bind _neighboringbuildingactorstodamageforcurrenttask (Variables|BuildingDamage|GetNeighboringBuildingActorsToDamageForCurrentTask))
  ;[0.1302 0 0.08458] StartNextBuildingActorsDamageTask: IF a task is NOT already running, start a new task to detect and damage building actors within the earthquake.
  (Networking|SwitchHasAuthority
    (:Authority
      (if (not (Variables|BuildingDamage|GetbBuildingActorsDamageTaskRunning))
        (Variables|BuildingDamage|SetbBuildingActorsDamageTaskRunning true)
        ;;[0 0.1458 0.1562] Remove the earthquake damage tag from all the building actors that were previously set to be damaged.
        (bind _outelementarray (Utilities|GetSubsetOfArraywithDelayLoop _buildingactorstodamageforcurrenttask (Variables|General|GetMaxNumOfActorsToLoopPerDelayDurationConst))
          (:Loop
            (General|RemoveActorTagfromActors _outelementarray (Variables|General|GetEarthquakeDamageBuildingActorTag)))
          (:Completed
            (Utilities|Array|Clear _buildingactorstodamageforcurrenttask)
            ;;[0.06436 0 0.1198] From the latest set of overlapped actors, find the building actors that qualify to being damaged by the earthquake.
            (bind _outelementarray_2 (Utilities|GetSubsetOfArraywithDelayLoop (Variables|AsyncOverlapActors|GetAsyncOverlappedActors) (Variables|General|GetMaxNumOfActorsToLoopPerDelayDurationConst))
              (:Loop
                (for _array_element _outelementarray_2
                  (Utilities|IsValid _array_element
                    (:Is%20Valid
                      (bind _asbuilding_static_mesh (Utilities|Casting|CastToBuildingSMActor _array_element))
                      (bind _bcansimulate (Forces|CanActorSimulateEarthquakePhysicsForces _asbuilding_static_mesh))
                      (if (not _bcansimulate)
                        (if (Building|WillRegisterwithStructuralGrid _asbuilding_static_mesh)
                          (if (not (Building|IsSupportedbyWorld _asbuilding_static_mesh))
                            (if (GameplayTags|DoesContainerMatchTagQuery (GameplayTags|GetOwnedGameplayTags _asbuilding_static_mesh :!class /Script/GameplayTags.GameplayTagAssetInterface) (Variables|BuildingDamage|GetQueryBuildingActorConsiderOnGround))
                              _))
                          (else
                            _))
                        (BuildingDamage|AddBuildingActortoDamage _asbuilding_static_mesh))))))
              (:Completed
                (AsyncOverlapActors|RemoveNullsfromAsyncOverlappedActors) ; When actors would be removed from AsyncOverlappedActors, instead their index in the array is set to null.\nThis is to maintain the indices for multi frame processing with GetSubsetOfArrayWithDelayLoop. Once the \nprocessing is complete, remove all the null values.\n
                ;;[0.1042 0 0.01125] Loop through the building actors to damage one set periodically until they all have been damaged for this task.
                ;; This implementation of breaking down the building actors into sets help prevent overloading the server with damaging a potentially large number of building actors on the same frame. Helps optimize network and gamethread.
                (Variables|BuildingDamage|SetBuildingActorsToDamageForCurrentTaskLoop (Variables|BuildingDamage|GetBuildingActorsToDamageForCurrentTask)) ; This array will be looped through.
                (Utilities|Array|Clear (Variables|BuildingDamage|GetNeighboringBuildingActorsToDamageForCurrentTask))
                (bind _outelementarray_3 (Utilities|GetSubsetOfArraywithDelayLoop (Variables|BuildingDamage|GetBuildingActorsToDamageForCurrentTaskLoop) (ScalableFloat|GetValueasInteger (Variables|BuildingDamage|GetMaxNumberOfBuildingActorsToDamageEverySet)) _returnvalue)
                  (:Loop
                    _)
                  (:Completed
                    (Utilities|Array|RemoveItem _neighboringbuildingactorstodamageforcurrenttask) ; Remove null entries
                    (if (Utilities|Array|IsValidIndex _neighboringbuildingactorstodamageforcurrenttask)
                      (Variables|BuildingDamage|SetBuildingActorsToDamageForCurrentTaskLoop (Variables|BuildingDamage|GetNeighboringBuildingActorsToDamageForCurrentTask))
                      (Utilities|FlowControl|Delay _returnvalue))
                    ;;[0 0.01661 0.1615] Once the building actors have all been damaged, wait for the desired duration before running a damage task again.
                    (Variables|BuildingDamage|SetbBuildingActorsDamageTaskRunning false)
                    (bind _returnvalue_2 (Utilities|Time|SetTimerbyEvent (EventDispatchers|CreateEvent :!event Custom|StartNextBuildingActorsDamageTask) (ScalableFloat|GetValueatLevel (Variables|BuildingDamage|GetBuildingActorsDamageTaskInterval))))
                    (Variables|BuildingDamage|SetBuildingActorsDamageTimerHandle _returnvalue_2))) ; Delay between every loop iteration for optimization \nsince the input array can be large.
                (for _array_element_2 _outelementarray_3
                  (bind _neighboringbuildingactorstodamage (BuildingDamage|DamageBuildingActor _array_element_2))
                  (Utilities|Array|AppendArray (Variables|BuildingDamage|GetNeighboringBuildingActorsToDamageForCurrentTask) _neighboringbuildingactorstodamage)))) ; Frame delay between every loop iteration for optimization \nsince the input array can be large.
          )) ; Frame delay between every loop iteration for optimization \nsince the input array can be large.
      ))))

(event Custom|AsyncOverlapTrace_Completion
  (params (WasCancelled :type Boolean))
  ;[0 0.01918 0.1354] AsyncOverlapTrace_Completion: Once the overlap trace is completed, remove any actors from the list that are now out of the earthquake radius and wait for desired interval before running another overlap trace.
  (Networking|SwitchHasAuthority
    (:Authority
      (Variables|Overlap|SetbAsyncOverlapTraceInProgress false)
      (AsyncOverlapActors|RemoveAsyncOverlappedActorsOutsideRadius)
      (if (not WasCancelled)
        (bind _returnvalue (Utilities|Time|SetTimerbyEvent (EventDispatchers|CreateEvent :!event RunAsyncOverlapTrace) (ScalableFloat|GetValueatLevel (Variables|Overlap|GetAsyncOverlapTraceUpdateInterval))))
        (Variables|Overlap|SetAsyncOverlapTraceTimerHandle _returnvalue)))))

(event Custom|HandleOverlappedActorDestroyed
  (params (DestroyedActor :type Actor%20Object%20Reference))
  ;[0.5 0 0.3383] HandleOverlappedActorDestroyed: When an overlapped actor is destroyed, treat the actor like it has ended overlap with the earthquake.
  (Networking|SwitchHasAuthority
    (:Authority
      (Utilities|IsValid DestroyedActor
        (:Is%20Valid
          (Game|UnbindEventfromOnDestroyed DestroyedActor (EventDispatchers|CreateEvent :!event Custom|HandleOverlappedActorDestroyed))
          (CallFunction|HandleActorEndAsyncOverlap DestroyedActor))))))

(event Custom|HandleActorBeginAsyncOverlap
  (params (Actor :type Actor%20Object%20Reference))
  ;[0.1244 0 0.1875] HandleActorBeginAsyncOverlap: Called when an actor overlaps starts overlapping the earthquake.
  (Networking|SwitchHasAuthority
    (:Authority
      (Utilities|IsValid Actor
        (:Is%20Valid
          (Forces|RegisterActorforEarthquakePhysicsForces Actor) ; Some actors can be affected by the earthquake's force.\nAdjust their physics settings and register to the earthquake \nphysics actor list IF the actor can receive the earthquake's forces.
          (Overlap|SetEnableDangerGameplayEffectonAI Actor true))))))

(event Custom|HandleActorEndAsyncOverlap
  (params (Actor :type Actor%20Object%20Reference))
  ;[0 0.1446 0.5104] HandleActorEndAsyncOverlap: Called when an actor is no longer overlapping the earthquake.
  (Networking|SwitchHasAuthority
    (:Authority
      (Utilities|IsValid Actor
        (:Is%20Valid
          (Forces|UnregisterActorfromEarthquakePhysicsForces Actor) ; Remove this actor from the \nphysics actor list IF it was registered.
          (Overlap|SetEnableDangerGameplayEffectonAI Actor))))))

(event AsyncOverlapActors|EventOnActorRemovedFromAsyncOverlappedActors
  (params (RemovedActor :type Actor%20Object%20Reference))
  ;[0.1879 0.5228 1] OnActorRemovedFromAsyncOverlappedActors - Called each time an actor is removed from AsyncOverlappedActors due to them no longer being within the earthquake radius
  (General|RemoveActorTagfromActors (Utilities|Array|MakeArray RemovedActor) (Variables|General|GetEarthquakeIdentificationName))
  (CallFunction|HandleActorEndAsyncOverlap RemovedActor))

(fn AddBuildingActorToDamage
  (params (BuildingActor :type Building%20Static%20Mesh%20Object%20Reference))
  ;[0.1137 0 0.1667] AddBuildingActorToDamage: Adds the input building actor to the building actors to damage list.
  (Utilities|FlowControl|Sequence
    (:then_0
      (Networking|SwitchHasAuthority
        (:Authority
          (Utilities|IsValid BuildingActor
            (:Is%20Valid
              (if (and (not (Utilities|Array|ContainsItem (Variables|/Script/Engine.Actor|GetTags BuildingActor) (Variables|General|GetEarthquakeDamageBuildingActorTag))) (not (Utilities|Array|ContainsItem (Variables|/Script/Engine.Actor|GetTags BuildingActor) "MilkshakeEarthquakeDamageImmune")))
                (Utilities|Array|Add (Variables|BuildingDamage|GetBuildingActorsToDamageForCurrentTask) BuildingActor)
                (Utilities|Array|Add (Variables|/Script/Engine.Actor|GetTags BuildingActor) (Variables|General|GetEarthquakeDamageBuildingActorTag))
                (LocalVariables|SetRetSuccess true)))))))
    (:then_1
      (return (LocalVariables|GetRetSuccess)))))

(fn AsyncOverlapTrace_ProcessHitActor
  (params (HitActor :type Actor%20Object%20Reference))
  (bind (_asfort_character_vehicle _bsuccess) (Utilities|Casting|CastToFortCharacterVehicle HitActor))
  (bind _tags (Variables|/Script/Engine.Actor|GetTags HitActor))
  (bind _earthquakeidentificationname (Variables|General|GetEarthquakeIdentificationName))
  ;[0 0.05149 0.07292] AsyncOverlapTraceProcessHitActor: When an actor is found in the overlap trace, check to see if it is already identified, and if it isn't add it to the overlapped actors list.
  (Networking|SwitchHasAuthority
    (:Authority
      (Utilities|IsValid HitActor
        (:Is%20Valid
          (if (not _bsuccess)
            (if (not (Utilities|Array|ContainsItem _tags _earthquakeidentificationname))
              (bind _bwithinradius (Overlap|IsActorWithinEarthquakeRadius HitActor)) ; Check distance here before adding to prevent actors on the border of the earthquake from\nbeing added and removed repeatedly. (Because they overlap, but their location isn't actually in\nthe radius.)
              (if _bwithinradius
                (Utilities|Array|Add _tags _earthquakeidentificationname)
                (Utilities|Array|Add (Variables|AsyncOverlapActors|GetAsyncOverlappedActors) HitActor)
                (Game|BindEventtoOnDestroyed HitActor (EventDispatchers|CreateEvent :!event HandleOverlappedActorDestroyed))
                (CallFunction|HandleActorBeginAsyncOverlap HitActor)))))))))

(fn AsyncOverlapTrace_ShouldCancelProgress
  ;[0 0.1354 0.1017] Only cancel the async overlap process IF the earthquake actor is being destroyed.
  (Networking|SwitchHasAuthority
    (:Authority
      (return (Game|IsActorBeingDestroyed)))
    (:Remote
      (return true))))

(fn CanActorSimulateEarthquakePhysicsForces
  (params (Actor :type Actor%20Object%20Reference))
  ;[0.07062 0.01058 0.08333] CanActorSimulateEarthquakePhysicsForces: Currently, only vehicle actors can simulate earthquake physics.
  (Utilities|FlowControl|Sequence
    (:then_0
      (Networking|SwitchHasAuthority
        (:Authority
          (if (ScalableFloat|GetValueasBool (Variables|Default|GetShouldApplyForceToPhysicsObjects))
            (bind _returnvalue (FortPhysicsSimulationLibrary|CanActorSimulatePhysics Actor))
            (if _returnvalue
              (Utilities|Casting|CastToFortAthenaVehicle Actor) ; Only works for vehicles at this time.
              (return true))))))
    (:then_1
      (return false))))

(fn Cheat_SpawnRadiusSphere
  (bind (_earthquakestartgametime_111_1f928ca54731f603a460a7b5e81d3536 _earthquakeendgametime_109_a139a81e4ea551f5bbd8c5ab7705588b _earthquakeradius_108_1e491f714c8af8bbeadee384f2595ad4) (Utilities|Struct|BreakMilkshakeEarthquakeRepInfoStruct (Variables|General|GetEarthquakeRepInfo)))
  ;[0 0 0] Cheat_SpawnRadiusSphere: Spawn a sphere actor in the world to represent the earthquake radius.
  (Networking|CanExecuteCosmeticEvents
    (:True
      (bind _splineactor (Variables|Movement|GetSplineActor)
        (:then
          (if (Cheat|CheatGetEarthquakeShow _splineactor :!class /Milkshake/Gameplay/BP_Milkshake_Spline_Earthquake.BP_Milkshake_Spline_Earthquake_C)
            _))
        (:else
          (bind _randomearthquakemanager (Variables|RandomEarthquake|GetRandomEarthquakeManager)
            (:then
              (if (Cheat|CheatGetEarthquakeShow _randomearthquakemanager :!class /Milkshake/Gameplay/BP_Milkshake_RandomEarthquakeManager.BP_Milkshake_RandomEarthquakeManager_C)
                _))
            (:else))))
      (bind _cheat_radiussphere (Variables|Cheat|GetCheat_RadiusSphere)
        (:then
          (Actor|DestroyActor _cheat_radiussphere))
        (:else
          _))
      (bind _returnvalue (Game|SpawnActorfromClass "/Milkshake/Gameplay/BP_Milkshake_EarthquakeDebug.BP_Milkshake_EarthquakeDebug_C" (Math|Conversions|ToTransform(Vector) (Transformation|GetActorLocation)) "AlwaysSpawn" "OverrideRootScale" 0 "" "0.000000,0.000000,0.000000" _earthquakeradius_108_1e491f714c8af8bbeadee384f2595ad4 "0.000000,0.000000,0.000000" 0))
      (Utilities|IsValid _returnvalue
        (:Is%20Valid
          (bind _output_get (Variables|Cheat|SetCheat_RadiusSphere _returnvalue))
          (Transformation|AttachActorToActor _output_get self "None" "SnapToTarget" "KeepWorld" "KeepWorld" true)
          (Class|BPMilkshakeEarthquakeDebug|DestroyComponents _returnvalue (Utilities|Array|MakeArray (Variables|/Milkshake/Gameplay/BP_Milkshake_EarthquakeDebug.BP_Milkshake_EarthquakeDebug_C|GetDebugText _returnvalue) (Variables|/Milkshake/Gameplay/BP_Milkshake_EarthquakeDebug.BP_Milkshake_EarthquakeDebug_C|GetDebugSphereRed _returnvalue) (Variables|/Milkshake/Gameplay/BP_Milkshake_EarthquakeDebug.BP_Milkshake_EarthquakeDebug_C|GetDebugSphereGreen _returnvalue))))))))

(fn DamageBuildingActor
  (params (BuildingActor :type Building%20Static%20Mesh%20Object%20Reference))
  (bind (_outactorgridindices _outneighboringactors) (Building|K2GetNeighboringBuildingActors (Building|GetBuildingStructuralSupportSystem) (Variables|/Script/FortniteGame.BuildingActor|GetBuildingType BuildingActor) (Transformation|GetActorLocation BuildingActor)))
  (bind _buildingstructuralsupportsystem (Building|GetBuildingStructuralSupportSystem))
  (bind (_x _y) (Math|Vector2D|BreakVector2D (Utilities|Struct|BreakMilkshakeEarthquakeInfoStruct (General|GetEarthquakeInfo))))
  ;[0 0.0597 0.06771] DamageBuildingActor: Look to damage the input building actor based on its distance from the earthquake.
  ; Also check to see if the neighboring building actors need to be added to the damage actor list.
  (Utilities|FlowControl|Sequence
    (:then_0
      (Networking|SwitchHasAuthority
        (:Authority
          ;;[0.164 0 0.1667] Check IF the building can be damaged.
          (Utilities|IsValid BuildingActor
            (:Is%20Valid
              (if (Variables|/Script/Engine.Actor|GetbCanBeDamaged BuildingActor)
                ;;[0.1823 0.03275 0.1979] IF this building actor is being damaged for the first time, there's a small chance it gets selected be damageable until it is destroyed.
                ;; This function call ends up deciding that.
                (BuildingDamage|SetDamageTypeGameplayEffectforBuildingActor BuildingActor)
                ;;[0 0.04876 0.07812] Get the normalized multiplier value to use based on the distance of the earthquake from the building actor to damage. Objects closer to the epicenter should take more damage and vice versa.
                (bind _alpha (General|GetNormalizedEarthquakeMagnitudeatLocation (Transformation|GetActorLocation)))
                (if (> _alpha 0)
                  (Utilities|FlowControl|Sequence
                    (:then_0
                      ;;[0.08463 0.1342 0.25] Apply damage to the building actor.
                      ;; Damage value is calculated as a factor of the max health of the building structure using a multiplier.
                      ;; Player built structures take full damage.
                      ;; Non-player built structures take damage based on their current health percentage specified through a curve.
                      ;; This is done to keep the world structures alive but cracked for as long as possible; otherwise, the world buildings collapse quickly and the whole thing looks odd. This implementation is subject to change based on feel.
                      (bind _output_get (LocalVariables|SetBuildingDamge (Math|Float|float*float (* (Health|GetMaxHealth BuildingActor :!class /Script/FortniteGame.BuildingActor) (Math|Random|RandomFloatinRange _x _y)) (BuildingDamage|GetBuildingDamageMultiplier BuildingActor) _alpha (ScalableFloat|GetValueatLevel (Variables|BuildingDamage|GetBuildingDamageMultiplier)) (ScalableFloat|GetValueatLevel (select (BuildingActor|IsPlayerBuilt BuildingActor) (Variables|Default|GetPlayerBuildingDamageMulitplier) (Variables|Default|GetNonPlayerBuildingDamageMulitplier))))))
                      (if (> _output_get (LocalVariables|GetMinimumBuildingDamageAllowed))
                        (Environment|ApplyGlobalEnvironmentDamagetoActor :DamageTargetActor BuildingActor :EnvironmentDamage _output_get :EnvironmentDamageGE (Variables|BuildingDamage|GetEarthquakeEnvironmentDamageGE))
                        (if (<= (Health|GetHealthPercent BuildingActor :!class /Script/FortniteGame.BuildingActor) 0)
                          (Fortnite|FortShippingLog self (Utilities|String|Append "DamageBuildingActor: Destroying building actor: " (Utilities|GetDisplayName BuildingActor))))))
                    (:then_1
                      ;;[0.2083 0 0.1005] This implementation handles damage spread through the building as the earthquake by default only damages the building actors that are touching the ground.
                      ;; Whenever the damaged building goes below a threshold health value. All of its neighboring building actors that are NOT detected by the earthquake yet are added to building actor to damage list.
                      ;; Note: The indentification name is added to the actors that are currently detected to be damaged by the earthquake actor.
                      (if (< (Building|GetBuildingHealthPercentage BuildingActor) (Variables|BuildingDamage|GetBuildingDamageSpreadThresholdHealthPercent))
                        (Utilities|IsValid _buildingstructuralsupportsystem
                          (:Is%20Valid
                            (if (Building|IsBuildingActorStructurallySupported _buildingstructuralsupportsystem BuildingActor)
                              (bind _buildingactors (BuildingDamage|GetNeighboringActorsasBuildingActors _outneighboringactors))
                              (for _array_element _buildingactors
                                (Utilities|IsValid _array_element
                                  (:Is%20Valid
                                    (bind _bwithinradius (Overlap|IsActorWithinEarthquakeRadius _array_element))
                                    (if _bwithinradius
                                      (bind _bsuccess (BuildingDamage|AddBuildingActortoDamage _array_element))
                                      (if _bsuccess
                                        (Utilities|Array|Add (LocalVariables|GetRetNeighboringBuildingActorsToDamage) _array_element)))))))))))))))))))
    (:then_1
      (return (LocalVariables|GetRetNeighboringBuildingActorsToDamage)))))

(fn GenerateNewForceDirectionForPhysicsActor
  (params (Actor :type Actor%20Object%20Reference) (PreviousForceDirection :type Vector))
  ;[0.1086 0.0264 0.1198] GenerateNewForceDirectionForPhysicsActor: Returns the opposite direction vector of the previous force direction.
  ; IF the input previous force direction is an empty vector (default), then this function returns a random direction selected between the input actor's forward or backward vector.
  ; This works well for vehicle actors but may need to be tweaked for other actor types.
  (Utilities|FlowControl|Sequence
    (:then_0
      (Networking|SwitchHasAuthority
        (:Authority
          (Utilities|IsValid Actor
            (:Is%20Valid
              (if (Utilities|Operators|Equal(==) :A PreviousForceDirection)
                (return (select (Math|Random|RandomBool) (Math|Vector|NegateVector (Transformation|GetActorForwardVector Actor)) (Transformation|GetActorForwardVector Actor)))
                (else
                  (return (Math|Vector|Normalize (Math|Vector|NegateVector PreviousForceDirection))))))))))
    (:then_1
      (return "0, 0, 0"))))

(fn GenerateNextForceDirectionUpdateTime
  ;[0 0.1742 0.3438] Returns a random game time value in the future calculated with the provided restrictions.
  (return (+ (Fortnite|Sync|Time|GetServerWorldTimeSeconds) (Math|Random|RandomFloatinRange (Variables|MilkshakeEarthquakeActor|GetMinTimeBeforeNextForceDirectionUpdate) (Variables|MilkshakeEarthquakeActor|GetMaxTimeBeforeNextForceDirectionUpdate)))))

(fn GetBuildingDamageGameplayEffectClassToUse
  (params (BuildingActor :type Building%20Actor%20Object%20Reference))
  (bind (_buildingdamagebasehealthpercentage_123_380af5a644d702698fe805b9517be16c _buildingdamagedestroychanceweight_120_af0ee3b646dba906f0732c8dc472022d) (Utilities|Struct|BreakMilkshakeEarthquakeInfoStruct (General|GetEarthquakeInfo)))
  (bind _buildingactor (Variables|GetBuildingActor))
  (bind _returnvalue (GameplayTags|GetOwnedGameplayTags _buildingactor :!class /Script/GameplayTags.GameplayTagAssetInterface))
  ;[0.02205 0 0.08333] GetBuildingDamageGameplayEffectClassToUse: Returns the gameplay effect that should be added to the building actor to help determine the damage curve the earthquake should use when determining the damage to apply to the building actor.
  (Utilities|FlowControl|Sequence
    (:then_0
      (Networking|SwitchHasAuthority
        (:Authority
          (Utilities|IsValid BuildingActor
            (:Is%20Valid
              ;;[0 0.03125 0.07812] Check for reasons in ordered priority.
              (Utilities|FlowControl|Sequence
                (:then_0
                  ;;;[1 1 1] Physics tree check. Only one physics tree can be destroyed per earthquake for perf reasons.
                  (if (GameplayTags|HasMatchingGameplayTag (Variables|GetBuildingActor) "(TagName=\"Building.Type.Prop.FellablePhysicsProp\")")
                    (if (< (Variables|BuildingDamage|GetNumOfPhysicsTreesAllowDestroy) (ScalableFloat|GetValueasInteger (Variables|BuildingDamage|GetMaxNumOfPhysicsTreesAllowDestroy)))
                      (Math|Integer|IncrementInt (Variables|BuildingDamage|GetNumOfPhysicsTreesAllowDestroy)
                        (:%20%20
                          (return (Variables|BuildingDamage|GetBuildingDamageDestroyGameplayEffectClass))))
                      (else
                        (return (Variables|BuildingDamage|GetBuildingDamageNeverDestroyGameplayEffectClass))))))
                (:then_1
                  ;;;[1 1 1] Check gameplay tag queries.
                  (if (and (not (GameplayTags|IsTagQueryEmpty (Variables|BuildingDamage|GetQueryBuildingActorAlwaysDestroyable))) (GameplayTags|DoesContainerMatchTagQuery _returnvalue (Variables|BuildingDamage|GetQueryBuildingActorAlwaysDestroyable)))
                    (return (Variables|BuildingDamage|GetBuildingDamageDestroyGameplayEffectClass))
                    (elif (and (not (GameplayTags|IsTagQueryEmpty (Variables|BuildingDamage|GetQueryBuildingActorNeverDestroyable))) (GameplayTags|DoesContainerMatchTagQuery _returnvalue (Variables|BuildingDamage|GetQueryBuildingActorNeverDestroyable)))
                      (return (Variables|BuildingDamage|GetBuildingDamageNeverDestroyGameplayEffectClass)))))
                (:then_2
                  ;;;[1 1 1] Building type check.
                  (if (Utilities|Array|ContainsItem (Variables|BuildingDamage|GetBuildingTypesNeverDestroy) (Variables|/Script/FortniteGame.BuildingActor|GetBuildingType (Variables|GetBuildingActor)))
                    (return (Variables|BuildingDamage|GetBuildingDamageNeverDestroyGameplayEffectClass))))
                (:then_3
                  ;;;[1 1 1] Return based on random weight generated.
                  (return (select (Math|Random|RandomBoolwithWeight (* _buildingdamagedestroychanceweight_120_af0ee3b646dba906f0732c8dc472022d (ScalableFloat|GetValueatLevel (Variables|BuildingDamage|GetBuildingDamageDestroyChanceWeightMultiplier)))) (Variables|BuildingDamage|GetBuildingDamageDestroyGameplayEffectClass) (Variables|BuildingDamage|GetBuildingDamageNeverDestroyGameplayEffectClass))))))))))
    (:then_1
      (return 0))))

(fn GetBuildingDamageMultiplier
  (params (BuildingActor :type Building%20Actor%20Object%20Reference))
  ;[0 0.03993 0.0625] GetBuildingDamageMultiplier: Based on the type of building actor, select the damage multiplier value from the respective curve variable.
  (Utilities|FlowControl|Sequence
    (:then_0
      (Networking|SwitchHasAuthority
        (:Authority
          (Utilities|IsValid BuildingActor
            (:Is%20Valid
              (Utilities|FlowControl|Sequence
                (:then_0
                  (LocalVariables|SetSelectedCurve (Variables|BuildingDamage|GetPlayerBuildingDamageCurve))
                  (if (GameplayTags|HasMatchingGameplayTag BuildingActor "(TagName=\"Milkshake.BA.D.D\")")
                    (LocalVariables|SetSelectedCurve (Variables|BuildingDamage|GetBuildingDamageDestroyCurve))
                    (elif (GameplayTags|HasMatchingGameplayTag BuildingActor "(TagName=\"Milkshake.BA.D.ND\")")
                      (LocalVariables|SetSelectedCurve (Variables|BuildingDamage|GetBuildingDamageNeverDestroyCurve)))))
                (:then_1
                  (bind _selectedcurve (LocalVariables|GetSelectedCurve)
                    (:then
                      (bind _output_get (LocalVariables|SetCurveValue (Math|Curves|GetFloatValue _selectedcurve (Health|GetHealthPercent BuildingActor :!class /Script/FortniteGame.BuildingActor))))
                      (if (> _output_get 0.01)
                        (return _output_get)))
                    (:else)))))))))
    (:then_1
      (return 0.0))))

(fn GetDangerAtLocation
  (params (Location :type Vector%20%28by%20ref%29 :default "0, 0, 0"))
  (bind _returnvalue (Transformation|GetActorLocation))
  (bind _returnvalue_2 (Math|Vector|Distance2DSquared(Vector) Location _returnvalue))
  (bind _returnvalue_3 (CallFunction|GetRadius))
  (bind _additionaldangerradius (Variables|AI|GetAdditionalDangerRadius))
  (bind _returnvalue_4 (+ _returnvalue_3 _additionaldangerradius))
  ;[0.01803 0.07137 0.1354] Calculate the danger as a float between 1 and 0 with a linear falloff from the center of the earthquake
  ;;[0.126 0.4758 0.8958] Check that the location is within the danger radius before performing any more operations
  (if (>= _returnvalue_2 (Math|Float|Square _returnvalue_4))
    (return 0.0)
    (else
      (return (Math|Float|MapRangeClamped (Math|Float|Sqrt _returnvalue_2) 0.0 _returnvalue_4 1.0)))))

(fn GetDangerSourceActive
  ;[0.002886 0.1705 0.2448] Returns if the earthquake is an active damage source.
  (return (and (Variables|General|GetbEarthquakeActive) (> (ScalableFloat|GetValueatLevel (Variables|AI|GetbAllowAIDanager)) 0))))

(fn GetDangerSourceBounds
  (return (- (Transformation|GetActorLocation) (+ (CallFunction|GetRadius) (Variables|AI|GetAdditionalDangerRadius))) (+ (+ (CallFunction|GetRadius) (Variables|AI|GetAdditionalDangerRadius)) (Transformation|GetActorLocation))))

(fn GetEarthquakeInfo
  ;[0.1451 0 0.2604] Return the earthquake info.
  (return (Variables|General|GetServerEarthquakeInfoStruct) (Variables|General|GetEarthquakeRepInfo)))

(fn GetEarthquakeTimeElapsed
  ;[0.1875 0 0.07527] Get the time elapsed from the timer handle that tracks the duration of the earthquake.
  (return (- (Fortnite|Sync|Time|GetServerWorldTimeSeconds) (Utilities|Struct|BreakMilkshakeEarthquakeRepInfoStruct (Variables|General|GetEarthquakeRepInfo)))))

(fn GetEventName
  ;[0.1739 0.4226 1] Returns the event name.
  (bind _owningearthquakecomponent (Variables|Default|GetOwningEarthquakeComponent)
    (:then
      (return (Class|BPMilkshakeEarthquakeManagerComponent|GetEventName _owningearthquakecomponent)))
    (:else
      (bind _randomearthquakespawner (Variables|RandomEarthquake|GetRandomEarthquakeSpawner)
        (:then
          (return (Class|BPMilkshakeRandomEarthquakeSpawner|GetSpawnerName _randomearthquakespawner)))
        (:else
          (return ""))))))

(fn GetNeighboringActorsAsBuildingActors
  (params (Info :type Building%20Neighboring%20Actor%20Info%20Structure))
  (bind (_neighboringwallinfos _neighboringfloorinfos _neighboringcentercellinfos) (Utilities|Struct|BreakBuildingNeighboringActorInfo Info))
  ;[0.08011 0 0.1198] Breaks down and returns the neighboring actor info struct as building actors.
  (for _array_element _neighboringwallinfos
    (bind _array_element_neighboringactor (Utilities|Struct|BreakNeighboringWallInfo _array_element))
    (Utilities|Array|Add (LocalVariables|GetReturnBuildingActors) _array_element_neighboringactor))
  (for _array_element_2 _neighboringfloorinfos
    (bind _array_element_neighboringactor_2 (Utilities|Struct|BreakNeighboringFloorInfo _array_element_2))
    (Utilities|Array|Add (LocalVariables|GetReturnBuildingActors) _array_element_neighboringactor_2))
  (for _array_element_3 _neighboringcentercellinfos
    (bind _array_element_neighboringactor_3 (Utilities|Struct|BreakNeighboringCenterCellInfo _array_element_3))
    (Utilities|Array|Add (LocalVariables|GetReturnBuildingActors) _array_element_neighboringactor_3))
  (return (LocalVariables|GetReturnBuildingActors)))

(fn GetNormalizedDistanceBetweenLocationAndEarthquake
  (params (Location :type Vector))
  ;[0.1347 0 0.1458] GetNormalizedDistanceBetweenLocationAndEarthquake: Returns 0.0, if the input location is at or outside the radius of the earthquake.
  ; Returns 1.0, if the input location is at the center/epicenter of the earthquake.
  (bind _distance (Class|MissionBlueprintFunctionLibrary|DistanceBetweenTwoVectors Location (Transformation|GetActorLocation)))
  (return (Math|Float|MapRangeClamped _distance 0.0 (CallFunction|GetRadius) 1.0)))

(fn GetNormalizedEarthquakeMagnitudeAtLocation
  (params (Location :type Vector))
  ;[0.09985 0.1458 0] GetNormalizedEarthquakeMagnitudeAtLocation: Returns the normalized magnitude of the earthquake at a given location.
  (bind _alpha (General|GetNormalizedDistanceBetweenLocationandEarthquake Location))
  (bind _value (General|GetNormalizedEarthquakeStrength))
  (return (Math|Float|Clamp(Float) (* _value (Math|Curves|GetFloatValue (Utilities|Struct|BreakMilkshakeEarthquakeRepInfoStruct (Variables|General|GetEarthquakeRepInfo)) _alpha)))))

(fn GetNormalizedEarthquakeStrength
  (bind (_earthquakestartgametime_111_1f928ca54731f603a460a7b5e81d3536 _earthquakeendgametime_109_a139a81e4ea551f5bbd8c5ab7705588b _earthquakeradius_108_1e491f714c8af8bbeadee384f2595ad4 _normalizedstrengthcurve_26_c1c0a94845e1806fb67b70b9bd380efe) (Utilities|Struct|BreakMilkshakeEarthquakeRepInfoStruct (Variables|General|GetEarthquakeRepInfo)))
  ;[0.02159 0.09663 0.1562] GetNormalizedEarthquakeStrength: Returns the normalized strength of the earthquake at the time of call.
  (return (Math|Curves|GetFloatValue _normalizedstrengthcurve_26_c1c0a94845e1806fb67b70b9bd380efe (Math|Float|SafeDivide (General|GetEarthquakeTimeElapsed) (- _earthquakeendgametime_109_a139a81e4ea551f5bbd8c5ab7705588b _earthquakestartgametime_111_1f928ca54731f603a460a7b5e81d3536)))))

(fn GetNormalizedMagnitudeAtLocation
  (params (Location :type Vector :default "0, 0, 0"))
  ;[0.09985 0.1458 0] GetNormalizedEarthquakeMagnitudeAtLocation: Returns the normalized magnitude of the earthquake at a given location.
  (bind _alpha (General|GetNormalizedDistanceBetweenLocationandEarthquake Location))
  (bind _value (General|GetNormalizedEarthquakeStrength))
  (return (Math|Float|Clamp(Float) (* _value (Math|Curves|GetFloatValue (Utilities|Struct|BreakMilkshakeEarthquakeRepInfoStruct (Variables|General|GetEarthquakeRepInfo)) _alpha)))))

(fn GetRadius
  (bind (_earthquakerepinfo_earthquakestartgametime_111_1f928ca54731f603a460a7b5e81d3536 _earthquakerepinfo_earthquakeendgametime_109_a139a81e4ea551f5bbd8c5ab7705588b _earthquakerepinfo_earthquakeradius_108_1e491f714c8af8bbeadee384f2595ad4) (Utilities|Struct|BreakMilkshakeEarthquakeRepInfoStruct (Variables|GetEarthquakeRepInfo_EarthquakeStartGameTime_111_1F928CA54731F603A460A7B5E81D3536)))
  (return _earthquakerepinfo_earthquakeradius_108_1e491f714c8af8bbeadee384f2595ad4))

(fn IsActorWithinEarthquakeRadius
  (params (Actor :type Actor%20Object%20Reference))
  ;[0.05163 0.1545 0.2708] Returns if the input actor is within the earthquake's radius.
  (Utilities|FlowControl|Sequence
    (:then_0
      (Utilities|IsValid Actor
        (:Is%20Valid
          (return (< (Math|Vector|DistanceSquared(Vector) (Transformation|GetActorLocation Actor) (Transformation|GetActorLocation)) (Math|Float|Square (CallFunction|GetRadius)))))))
    (:then_1
      (return false))))

(fn IsDebugEnabled
  ;[0.1341 0 0.224] Returns if debug is enabled.
  ; Debug is only allowed in PIE.
  (Utilities|FlowControl|Sequence
    (:then_0
      (bind _splineactor (Variables|Movement|GetSplineActor)
        (:then
          (return (Editor-Debug|IsDebugEnabled _splineactor :!class /Milkshake/Gameplay/BP_Milkshake_Spline_Earthquake.BP_Milkshake_Spline_Earthquake_C) (.bDrawDebugOutlines (Editor-Debug|IsDebugEnabled _splineactor :!class /Milkshake/Gameplay/BP_Milkshake_Spline_Earthquake.BP_Milkshake_Spline_Earthquake_C))))
        (:else
          (bind _randomearthquakespawner (Variables|RandomEarthquake|GetRandomEarthquakeSpawner)
            (:then
              (return (Editor-Debug|IsDebugEnabled _randomearthquakespawner :!class /Milkshake/Gameplay/BP_Milkshake_RandomEarthquakeSpawner.BP_Milkshake_RandomEarthquakeSpawner_C) (.bDrawDebugOutlines (Editor-Debug|IsDebugEnabled _randomearthquakespawner :!class /Milkshake/Gameplay/BP_Milkshake_RandomEarthquakeSpawner.BP_Milkshake_RandomEarthquakeSpawner_C))))
            (:else)))))
    (:then_1
      (return false false))))

(fn IsVehicleOnGround
  (params (Vehicle :type Fort%20Athena%20Vehicle%20Object%20Reference))
  ;[0 0.224 0.1675] Returns if the input vehicle is on the ground.
  (Utilities|IsValid Vehicle
    (:Is%20Valid
      (return (or (Vehicle|IsAsleep Vehicle) (and (Vehicle|WheelsonGround Vehicle) (Class|FortVehicleInterface|IsTouchingGround Vehicle)))))
    (:Is%20Not%20Valid
      (return false))))

(fn OnRep_EarthquakeRepInfo
  ;[0 0.154 0.2969] OnRep_EarthquakeRepInfo: Called once the earthquake rep info variable is updated.
  ; It is important to note that this function is called on both server and client.
  (CallFunction|BeginEarthquake))

(fn RegisterActorForEarthquakePhysicsForces
  (params (Actor :type Actor%20Object%20Reference))
  (bind _earthquakephysicsactortag (Variables|General|GetEarthquakePhysicsActorTag))
  ;[0.3412 0 0.4323] RegisterActorForEarthquakePhysicsForces: IF the input actor passes the criteria to simulate earthquake physics, update its physics settings and add it to the physics actor list.
  (if (ScalableFloat|GetValueasBool (Variables|Default|GetShouldApplyForceToPhysicsObjects))
    (Utilities|IsValid Actor
      (:Is%20Valid
        (if (not (Actor|ActorHasTag Actor _earthquakephysicsactortag))
          (bind _bcansimulate (Forces|CanActorSimulateEarthquakePhysicsForces Actor))
          (if _bcansimulate
            (Utilities|Array|Add (Variables|/Script/Engine.Actor|GetTags Actor) _earthquakephysicsactortag)
            (bind _returnvalue (FortPhysicsSimulationLibrary|GetPhysicsObjectSimSize Actor))
            (Utilities|Array|Add (Variables|MilkshakeEarthquakeActor|GetOverlappedPhysicsActorInfos) (Utilities|Struct|MakeMilkshakeEarthquakePhysicsActorInfo Actor (+ (Fortnite|Sync|Time|GetServerWorldTimeSeconds) (LocalVariables|GetInitialMinForceDirectionUpdateTime)) (Math|Random|RandomBool) _returnvalue))))))))

(fn RemoveActorTagFromActors
  (params (Actors :type Array%20of%20Actor%20Object%20References) (ActorTag :type Name%20%28by%20ref%29))
  ;[0.00978 0.02712 0.06771] Remove an actor tag from the input actors.
  (Networking|SwitchHasAuthority
    (:Authority
      (for _array_element Actors
        (Utilities|IsValid _array_element
          (:Is%20Valid
            (Utilities|Array|RemoveItem (Variables|/Script/Engine.Actor|GetTags _array_element) ActorTag)))))))

(fn RunAsyncOverlapTrace
  ;[0 0.07652 0.1719] RunAsyncOverlapTrace: Run an async overlap trace to find all the actors that are within the earthquake radius.
  (Networking|SwitchHasAuthority
    (:Authority
      (if (not (Variables|Overlap|GetbAsyncOverlapTraceInProgress))
        (Variables|Overlap|SetbAsyncOverlapTraceInProgress true)
        (Class|AsyncOverlapQueue|AsyncSphereOverlapActors :Location (Transformation|GetActorLocation) :Radius (CallFunction|GetRadius) :ShouldCancelProcess (EventDispatchers|CreateEvent :!event AsyncOverlapTrace_ShouldCancelProgress) :ProcessHitActorFunction (EventDispatchers|CreateEvent :!event AsyncOverlapTrace_ProcessHitActor) :CompletionFunction (EventDispatchers|CreateEvent :!event AsyncOverlapTrace_Completion) :TraceChannel "ECC_Visibility")))))

(fn SetDamageTypeGameplayEffectForBuildingActor
  (params (BuildingActor :type Building%20Static%20Mesh%20Object%20Reference))
  ;[0 0.05168 0.2292] SetDamageTypeGameplayEffectForBuildingActor: If NOT a player built building, there's a slight chance to mark this actor to take damage from the earthquake until it is destroyed.
  ; This is done by using a damage value curve that continues damaging the building actor even if its health is below a certain threshold.
  (Networking|SwitchHasAuthority
    (:Authority
      (Utilities|IsValid BuildingActor
        (:Is%20Valid
          (if (and (not (BuildingActor|IsPlayerBuilt BuildingActor)) (not (GameplayTags|HasAnyMatchingGameplayTags BuildingActor (GameplayTags|MakeLiteralGameplayTagContainer "(GameplayTags=((TagName=\"Milkshake.BA.D.D\"),(TagName=\"Milkshake.BA.D.ND\")))"))))
            (bind _returnvalue (BuildingDamage|GetBuildingDamageGameplayEffectClasstoUse BuildingActor))
            (if (Utilities|IsValidClass _returnvalue)
              (Fortnite|ApplyGameplayEffectsToActor BuildingActor (Utilities|Array|MakeArray _returnvalue) (0.0 0.0 0.0) false 0 0 "()"))))))))

(fn SetEnableDangerGameplayEffectOnAI
  (params (Actor :type Actor%20Object%20Reference) (bEnable :type Boolean))
  (bind _insidedangereffect (Variables|AI|GetInsideDangerEffect))
  ;[0.1932 0 0.2188] SetEnableDangerGameplayEffectOnAI: Add or remove gameplay effect that enables danger effect on the overlapping AI pawns.
  (Networking|SwitchHasAuthority
    (:Authority
      (bind _asfort_aipawn (Utilities|Casting|CastToFortAIPawn Actor))
      (bind _aipawnabilitysystemcomponent (Variables|/Script/FortniteGame.FortAIPawn|GetAIPawnAbilitySystemComponent _asfort_aipawn)
        (:then
          (if bEnable
            (GameplayEffects|ApplyGameplayEffectToSelf :self _aipawnabilitysystemcomponent :GameplayEffectClass _insidedangereffect :Level 0.0)
            (else
              (GameplayEffects|RemoveActiveGameplayEffectbySourceEffect _aipawnabilitysystemcomponent _insidedangereffect 0 -1))))
        (:else)))))

(fn SetRegisterToPlayerControllerComponents
  (params (bRegister :type Boolean))
  ;[0 0.2747 0.3802] Find all the milkshake controller components in the world and register/unregister this earthquake to/from those components.
  (Networking|CanExecuteCosmeticEvents
    (:True
      (for _array_element (Fortnite|GetLocalFortPlayerControllers true true)
        (Utilities|IsValid _array_element
          (:Is%20Valid
            (bind _output_get (LocalVariables|SetComponent (Actor|GetComponentbyClass _array_element "/Milkshake/Gameplay/BP_MilkshakeControllerComponent.BP_MilkshakeControllerComponent_C")))
            (Utilities|IsValid _output_get
              (:Is%20Valid
                (if bRegister
                  (Fortnite|FortShippingLog self (Utilities|String|Append "SetRegisterToPlayerControllerComponents: Registering earthquake to controller: " (Utilities|GetDisplayName _array_element)))
                  (Class|MilkshakeControllerComponent|RegisterEarthquake (LocalVariables|GetComponent) self)
                  (else
                    (Fortnite|FortShippingLog self (Utilities|String|Append "SetRegisterToPlayerControllerComponents: Unregistering earthquake from controller: " (Utilities|GetDisplayName _array_element)))
                    (Class|MilkshakeControllerComponent|UnregisterEarthquake (LocalVariables|GetComponent) self)))))))))))

(fn SetReplicatedEarthquakeInfo
  (bind (_enabled_104_9e9bd6984687100ae0fd3cb84b7e6a76 _calendareventratio_116_db14dea24c94389f1ded8fad08f78569 _earthquakesplinemovenormalizedlength_110_2e9091be45b8e7ff70c161ae980c69a9 _delaybetweenearthquakes_91_c25849394c195bda2b398d9bec9e0b9d _numberofearthquakes_93_299c58ad466aba23c371aeb841d710a5 _earthquakeduration_95_a139a81e4ea551f5bbd8c5ab7705588b _earthquakeradius_98_1e491f714c8af8bbeadee384f2595ad4 _buildingdamagebasehealthpercentage_123_380af5a644d702698fe805b9517be16c _buildingdamagedestroychanceweight_120_af0ee3b646dba906f0732c8dc472022d _safezonephase_106_f9a528fa4c9cfaf4c009fd820928814e _normalizedstrengthcurve_26_c1c0a94845e1806fb67b70b9bd380efe _normalizeddistancefalloffcurve_27_a4fe975840a1fa7e6d0ebb85fdddfe6b) (Utilities|Struct|BreakMilkshakeEarthquakeInfoStruct (Variables|General|GetServerEarthquakeInfoStruct)))
  (bind (_earthquakeduration_95_a139a81e4ea551f5bbd8c5ab7705588b_x _earthquakeduration_95_a139a81e4ea551f5bbd8c5ab7705588b_y) (Math|Vector2D|BreakVector2D _earthquakeduration_95_a139a81e4ea551f5bbd8c5ab7705588b))
  (bind (_earthquakeradius_98_1e491f714c8af8bbeadee384f2595ad4_x _earthquakeradius_98_1e491f714c8af8bbeadee384f2595ad4_y) (Math|Vector2D|BreakVector2D _earthquakeradius_98_1e491f714c8af8bbeadee384f2595ad4))
  ;[0 0.0572 0.2188] SetReplicatedEarthquakeInfo: Called from the Earthquake Spline after spawning this actor. It provides the necessary information to create the desired earthquake effects.
  (Networking|SwitchHasAuthority
    (:Authority
      (bind _output_get (Variables|General|SetEarthquakeRepInfo (Utilities|Struct|MakeMilkshakeEarthquakeRepInfoStruct (Fortnite|Sync|Time|GetServerWorldTimeSeconds) (+ (Fortnite|Sync|Time|GetServerWorldTimeSeconds) (Math|Random|RandomFloatinRange _earthquakeduration_95_a139a81e4ea551f5bbd8c5ab7705588b_x _earthquakeduration_95_a139a81e4ea551f5bbd8c5ab7705588b_y)) (Math|Random|RandomFloatinRange _earthquakeradius_98_1e491f714c8af8bbeadee384f2595ad4_x _earthquakeradius_98_1e491f714c8af8bbeadee384f2595ad4_y) _normalizedstrengthcurve_26_c1c0a94845e1806fb67b70b9bd380efe _normalizeddistancefalloffcurve_27_a4fe975840a1fa7e6d0ebb85fdddfe6b)))
      (Fortnite|FortShippingLog self (Utilities|String|Append "SetReplicatedEarthquakeInfo: Calendar Event " (General|GetEventName) ". Earthquake Radius: " (Utilities|String|ToString(Float) (.EarthquakeRadius_108_1E491F714C8AF8BBEADEE384F2595AD4 (Utilities|Struct|BreakMilkshakeEarthquakeRepInfoStruct _output_get))) ". Earthquake Duration: " (Utilities|String|ToString(Float) (- (.EarthquakeEndGameTime_109_A139A81E4EA551F5BBD8C5AB7705588B (Utilities|Struct|BreakMilkshakeEarthquakeRepInfoStruct _output_get)) (Utilities|Struct|BreakMilkshakeEarthquakeRepInfoStruct _output_get))))))))

(fn SetupUniqueActorTagNameVariables
  ;[0.1146 0 0.1085] SetupUniqueActorTagNameVariables: Setup the unique name variables that are going to be used by this earthquake.
  (Networking|SwitchHasAuthority
    (:Authority
      (bind _returnvalue (GUID|GenerateGuidfromObject self true))
      (if (Guid|IsValidGuid _returnvalue)
        (Variables|General|SetEarthquakeIdentificationName (Utilities|String|StringToName (Guid|ToString(Guid) _returnvalue))) ; Used to identify the actor that is overlapped by this earthquake.
        (Variables|General|SetEarthquakeDamageBuildingActorTag (Utilities|String|StringToName (Utilities|String|Append (Guid|ToString(Guid) _returnvalue) (Variables|General|GetEarthquakeDamageBuildingActorTagSuffix)))) ; Used to identify the building actor that is taking damage from this earthquake.
        (Variables|General|SetEarthquakePhysicsActorTag (Utilities|String|StringToName (Utilities|String|Append (Guid|ToString(Guid) _returnvalue) (Variables|General|GetEarthquakePhysicsActorTagSuffix)))) ; Used to identify physics actors that are affected by the earthquake.
      ))))

(fn SpawnRadialForceActor
  (bind (_earthquakestartgametime_111_1f928ca54731f603a460a7b5e81d3536 _earthquakeendgametime_109_a139a81e4ea551f5bbd8c5ab7705588b) (Utilities|Struct|BreakMilkshakeEarthquakeRepInfoStruct (Variables|General|GetEarthquakeRepInfo)))
  ;[0 0.1342 0.1771] On the client, spawn the radial force actor responsible for handling the tree/foilage (material and effects) movements in the area.
  ; Attach this actor to the earthquake so that it moves with the earthquake actor.
  (Networking|CanExecuteCosmeticEvents
    (:True
      (if (ScalableFloat|GetValueasBool (Variables|Default|GetShouldUseRadialForceFieldSystem))
        (bind _returnvalue (Game|SpawnActorfromClass "/Milkshake/Art/FX/BP_MilkshakeFieldSystemActor.BP_MilkshakeFieldSystemActor_C" (Math|Transform|MakeTransform (Transformation|GetActorLocation)) "AlwaysSpawn" "MultiplyWithRoot" 0 self (Variables|RadialForce|GetRadialForceStrength) (CallFunction|GetRadius) _earthquakeendgametime_109_a139a81e4ea551f5bbd8c5ab7705588b (* (- _earthquakeendgametime_109_a139a81e4ea551f5bbd8c5ab7705588b _earthquakestartgametime_111_1f928ca54731f603a460a7b5e81d3536) (Variables|RadialForce|GetRadialForceStartAndEndDurationRatio)) (* (- _earthquakeendgametime_109_a139a81e4ea551f5bbd8c5ab7705588b _earthquakestartgametime_111_1f928ca54731f603a460a7b5e81d3536) (Variables|RadialForce|GetRadialForceStartAndEndDurationRatio)) 0))
        (bind _output_get (Variables|RadialForce|SetRadialForce _returnvalue))
        (Utilities|IsValid _output_get
          (:Is%20Valid
            (Transformation|AttachActorToActor _returnvalue self "None" "SnapToTarget" "KeepRelative" "KeepRelative" false)))))))

(fn TickDebugEarthquake
  (bind _returnvalue (Transformation|GetActorLocation))
  (bind (_benabled _bdrawdebugoutlines) (General|IsDebugEnabled))
  ;[0 0.05729 0.01455] Handle drawing and printing debug information for this earthquake actor.
  (if _benabled
    (LocalVariables|SetbDrawDebugOutlines _bdrawdebugoutlines)
    (Utilities|FlowControl|Sequence
      (:then_0
        ;;[0.2214 0 0.375] Draw debug spheres to represent the earthquake epiccenter and radius.
        (if (LocalVariables|GetbDrawDebugOutlines)
          (Rendering|Debug|DrawDebugSphere _returnvalue (CallFunction|GetRadius) 80 "(R=0.000000,G=1.000000,B=0.374239,A=1.000000)" 0.0 2e+01)
          (Rendering|Debug|DrawDebugSphere _returnvalue 1e+02 80 "(R=1.000000,G=0.000000,B=0.784090,A=1.000000)" 0.0 2e+01)))
      (:then_1
        ;;[0.2031 0.07715 0.09336] On the screen, print the number of overlapped actors by the earthquake.
        (Development|PrintString (Utilities|String|Append "Earthquake: " (Utilities|GetObjectName self) " : NumOfOverlappedActors : " (Utilities|String|ToString(Integer) (Utilities|Array|Length (Variables|AsyncOverlapActors|GetAsyncOverlappedActors)))) true false (LocalVariables|GetGenericActorDebugBoxColor) 0.0)
        ;;[0.1875 0 0.1638] On the screen, print the number of physics overlapped actors by the earthquake.
        (Development|PrintString (Utilities|String|Append "Earthquake: " (Utilities|GetObjectName self) " : NumOfOverlappedPhysicsActors : " (Utilities|String|ToString(Integer) (Utilities|Array|Length (Variables|MilkshakeEarthquakeActor|GetOverlappedPhysicsActorInfos)))) true false (LocalVariables|GetPhysicsActorDebugBoxColor) 0.0)
        ;;[0.1875 0 0.1638] On the screen, print the number of actors to damage in the current task.
        (Development|PrintString (Utilities|String|Append "Earthquake: " (Utilities|GetObjectName self) " : NumOfActorsToDmgInCurrentTaskLoop : " (Utilities|String|ToString(Integer) (Utilities|Array|Length (Variables|BuildingDamage|GetBuildingActorsToDamageForCurrentTaskLoop)))) true false (LocalVariables|GetGenericActorDebugBoxColor) 0.0)
        ;;[0.1875 0 0.1638] On the screen, print the number of actors to damage in the current task.
        (Development|PrintString (Utilities|String|Append "Earthquake: " (Utilities|GetObjectName self) " : NumOfActorsToDmgInCurrentTask : " (Utilities|String|ToString(Integer) (Utilities|Array|Length (Variables|BuildingDamage|GetBuildingActorsToDamageForCurrentTask)))) true false (LocalVariables|GetGenericActorDebugBoxColor) 0.0))
      (:then_2
        ;;[0.07219 0.008355 0.07292] Draw debug boxes around the actors that are within the earthquake radius.
        (if (LocalVariables|GetbDrawDebugOutlines)
          (for _array_element (Variables|AsyncOverlapActors|GetAsyncOverlappedActors)
            (Utilities|IsValid _array_element
              (:Is%20Valid
                (Rendering|Debug|DrawDebugBox (Transformation|GetActorLocation _array_element) "100.000000,100.000000,100.000000" (Math|Color|SelectColor (LocalVariables|GetPhysicsActorDebugBoxColor) (LocalVariables|GetGenericActorDebugBoxColor) (Actor|ActorHasTag _array_element (Variables|General|GetEarthquakePhysicsActorTag))) (Transformation|GetActorRotation _array_element) 0.0 5.0)
                (bind _asbuilding_static_mesh (Utilities|Casting|CastToBuildingSMActor _array_element))
                (if (Utilities|Array|ContainsItem (Variables|BuildingDamage|GetBuildingActorsToDamageForCurrentTask) _asbuilding_static_mesh)
                  (Rendering|Debug|DrawDebugBox :Center (Transformation|GetActorLocation _asbuilding_static_mesh) :Extent "100.000000,100.000000,100.000000" :LineColor "(R=1.000000,G=0.000000,B=0.095176,A=1.000000)" :Duration 0.0 :Thickness 5.0 :DepthPriority "World"))))))))))

(fn TickLifetime
  ;[0 0.08024 0.1302] Ticks the lifetime of the earthquake actor.
  ; IF the server time is greater or equal to the set up EndGameTime, then destroy the earthquake actor.
  (if (>= (Fortnite|Sync|Time|GetServerWorldTimeSeconds) (Utilities|Struct|BreakMilkshakeEarthquakeRepInfoStruct (Variables|General|GetEarthquakeRepInfo)))
    (Actor|DestroyActor)))

(fn UnregisterActorFromEarthquakePhysicsForces
  (params (Actor :type Actor%20Object%20Reference))
  (bind _overlappedphysicsactorinfos (Variables|MilkshakeEarthquakeActor|GetOverlappedPhysicsActorInfos))
  (bind _earthquakephysicsactortag (Variables|General|GetEarthquakePhysicsActorTag))
  ;[0.1875 0.05995 0.1139] UnregisterActorFromEarthquakePhysicsForces: Remove the input actor from the list of actors that is setup to receive the earthquake's physics forces.
  (Networking|SwitchHasAuthority
    (:Authority
      (if (ScalableFloat|GetValueasBool (Variables|Default|GetShouldApplyForceToPhysicsObjects))
        (Utilities|IsValid Actor
          (:Is%20Valid
            (if (Actor|ActorHasTag Actor _earthquakephysicsactortag)
              (General|RemoveActorTagfromActors (Utilities|Array|MakeArray Actor) _earthquakephysicsactortag)
              ;;[0.01458 0.07893 0.1042] Loop backwards through the list and remove the elements by index.
              (bind (_arrayindex _arrayelement) (Utilities|Array|ReverseforEachLoop _overlappedphysicsactorinfos)
                (:LoopBody
                  (if (== (Utilities|Struct|BreakMilkshakeEarthquakePhysicsActorInfo _arrayelement) Actor)
                    (Utilities|Array|RemoveIndex _overlappedphysicsactorinfos _arrayindex)))))))))))

(fn UpdateQuakeFXParams
  ;[0.05592 0.2124 0.25] UpdateQuakeFXParams: Update the parameter on the actor FX
  (Networking|CanExecuteCosmeticEvents
    (:True
      (bind _value (General|GetNormalizedEarthquakeStrength))
      (Utilities|FlowControl|Sequence
        (:then_0
          (bind _quakefx (Variables|Default|GetQuakeFX)
            (:then
              (Effects|Components|ParticleSystem|SetFloatParameter _quakefx "QuakeMagnitude" _value)
              (Effects|Components|ParticleSystem|SetFloatParameter _quakefx "QuakeRadius" (CallFunction|GetRadius))
              (Transformation|SetWorldLocation _quakefx (Transformation|GetActorLocation) false false))
            (:else)))
        (:then_1
          (bind _earthquake_audio_component (Variables|Audio|GetEarthquake%20Audio%20Component)
            (:then
              (Audio|Parameter|SetFloatParameter _earthquake_audio_component (Variables|Audio|GetEarthquakeAudioParameterName) _value :!class /Script/AudioExtensions.AudioParameterControllerInterface))
            (:else)))))))

(fn ConstructionScript)
