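; Debug actor that visualizes an earthquake in the level as three colored spheres and a text label showing the earthquake's name.
; The construction script disables collision, places the red and green spheres, tints all three through a material vector parameter, and converts each radius to scale via RadiusToScaleRatio.
; The label scales with the outer sphere and sits half its radius above the origin.
; DestroyComponents lets the owner remove this actor's components.
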(fn DestroyComponents
  (params (Components :type Array%20of%20Actor%20Component%20Object%20References))
  ;[0 0.2135 0.1925] Destory in the input components;
  ; Used to send in components belonging to this actor so that the owner can destroy it (blueprint friendly)
  (Components|DestroyComponent Components))

(fn ConstructionScript
  (bind _radiustoscaleratio (Variables|Default|GetRadiusToScaleRatio))
  (bind _sphereouterradius (Variables|Default|GetSphereOuterRadius))
  (bind _returnvalue (/ _sphereouterradius _radiustoscaleratio))
  (bind _spherematerialcolorparamname (LocalVariables|GetSphereMaterialColorParamName))
  (bind _debugtext (Variables|Default|GetDebugText))
  ;[0.01016 0.1354 0] Construction Script: Update the values of the sphere and text components so that they display the corresponding information for the desired earthquake.
  (|CallParentFunction)
  (Collision|SetActorEnableCollision false)
  ;;[0.09896 0.04371 0.07746] Setup Sphere Green and Red locations
  (Transformation|SetWorldLocation (Variables|Default|GetDebugSphereRed) (Variables|SphereRed|GetSphereRed_Location) false false)
  (Transformation|SetWorldLocation (Variables|Default|GetDebugSphereGreen) (Variables|SphereGreen|GetSphereGreen_Location) false false)
  ;;[0 0.261 1] Setup sphere colors
  (Rendering|Material|SetVectorParameterValueonMaterials (Variables|Default|GetDebugSphereGreen) _spherematerialcolorparamname (Math|Conversions|ToVector(LinearColor) (LocalVariables|GetGreenColor)))
  (Rendering|Material|SetVectorParameterValueonMaterials (Variables|Default|GetDebugSphereOuter) _spherematerialcolorparamname (Math|Conversions|ToVector(LinearColor) (LocalVariables|GetSphereOuterColor)))
  (Rendering|Material|SetVectorParameterValueonMaterials (Variables|Default|GetDebugSphereRed) _spherematerialcolorparamname (Math|Conversions|ToVector(LinearColor) (LocalVariables|GetRedColor)))
  ;;[0 0.04676 0.151] Adjust the scale of the sphere component so that they represent the required radius value.
  (Transformation|SetWorldScale3D (Variables|Default|GetDebugSphereOuter) (Math|Conversions|ToVector(Float) _returnvalue))
  (Transformation|SetWorldScale3D (Variables|Default|GetDebugSphereRed) (Math|Conversions|ToVector(Float) (/ (Variables|SphereRed|GetSphereRed_Radius) (Variables|Default|GetRadiusToScaleRatio))))
  (Transformation|SetWorldScale3D (Variables|Default|GetDebugSphereGreen) (Math|Conversions|ToVector(Float) (/ (Variables|SphereGreen|GetSphereGreen_Radius) (Variables|Default|GetRadiusToScaleRatio))))
  ;;[0 0.0584 0.151] Update the text component size, location, and text value. The text shows the name (key) of the earthquake.
  (Rendering|Components|TextRender|SetXScale _debugtext _returnvalue)
  (Rendering|Components|TextRender|SetYScale _debugtext _returnvalue)
  (Rendering|Components|TextRender|SetText _debugtext (Variables|Default|GetEarthquakeText))
  (Transformation|SetRelativeLocation _debugtext (Math|Vector|MakeVector 0.0 0.0 (/ (Variables|Default|GetSphereOuterRadius) 2.0)) false false))
