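; Server-authoritative Fortnite manager actor that spawns random earthquakes during a match while its calendar event is active, gated by hotfix values.
; A timer polls calendar status, starting or stopping spawn checks limited by game phase, per-match total and randomized per-safe-zone concurrency caps.
; An EQS query finds spawners, the one nearest most players spawns the quake, then a safe-zone-scaled cooldown runs.
; Cheat events support forced spawning and debug display.
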(event EventBeginPlay
  (bind _brandomearthquakesenabled (Variables|Hotfix|GetbRandomEarthquakesEnabled))
  (bind _returnvalue (ScalableFloat|GetValueatLevel _brandomearthquakesenabled))
  (bind _returnvalue_2 (Class|FortDeterminismConfig|IsUsingNoMilkshake))
  (bind _bmilkshakeenabled (Variables|Hotfix|GetbMilkshakeEnabled))
  (bind _returnvalue_3 (ScalableFloat|GetValueatLevel _bmilkshakeenabled))
  (bind _returnvalue_4 (not _returnvalue_2))
  (bind _returnvalue_5 (> _returnvalue 0))
  (bind _returnvalue_6 (> _returnvalue_3 0))
  (bind _returnvalue_7 (Math|Boolean|ANDBoolean _returnvalue_6 _returnvalue_5 _returnvalue_4))
  ;[0 0.2396 0] BeginPlay: Initialize the actor, check if it is hotfixed off.
  (|CallParentFunction)
  (Networking|SwitchHasAuthority
    (:Authority
      (Class|FortAsyncActionCurrentPlaylistReady|PlaylistDataReadyAsync
        (:OnReady
          (Fortnite|FortShippingLog self (Utilities|String|Append "BeginPlay: Spawned. Hotfix: " (Utilities|String|ToString(Boolean) _returnvalue_7)))
          (if _returnvalue_7
            ;;[0 0.1406 0.07617] Start a timer to track the calendar event status.
            (Utilities|Time|SetTimerbyEvent (EventDispatchers|CreateEvent :!event UpdateCalendarEventStatus) (Variables|Default|GetCalendarEventStatusUpdateInterval) true false 0.2 0.2)
            ;;[0 0.04691 0.2135] Cache the maximum number of random earthquakes that are allowed to spawn in the match.
            (bind _output_get (Variables|Default|SetNumOfRandomEarthquakesToSpawn (Math|Random|RandomIntegerinRange (ScalableFloat|GetValueasInteger (Variables|Hotfix|GetNumOfRandomEarthquakesPerMatch)) (ScalableFloat|GetValueasInteger (Variables|Hotfix|GetNumOfRandomEarthquakesPerMatch) 1.0))))
            (Fortnite|FortShippingLog self (Utilities|String|Append "BeginPlay: NumOfRandomEarthquakesToSpawn: " (Utilities|String|ToString(Integer) _output_get)))
            (CallFunction|RegistertoActorRegistrySubsystemDelegates) ; Used to find out all the random earthquake spawners in the world.
          ))) ; Wait until the playlist data is ready before checking hotfix values
    )))

(event Custom|RandomEarthquakesEventStart
  ;[0 0.1533 0.1927] RandomEarthquakesEventStart: This function is called whenever the random earthquake calendar event becomes active.
  (Networking|SwitchHasAuthority
    (:Authority
      (Fortnite|FortShippingLog self "RandomEarthquakesEventStart: Invoked.")
      (CallFunction|StartEarthquakeSpawnCheckTimer)
      ;;[0.1771 0.1505 0] Start a timer that refreshes the maximum number of earthquakes that are allowed to spawn at once.
      (bind _returnvalue (Utilities|Time|SetTimerbyEvent (EventDispatchers|CreateEvent :!event UpdateMaxNumberOfEarthquakesAtOnceValues) (Variables|Default|GetUpdateMaxNumberOfEarthquakesAtOnceInterval) true))
      (Variables|Default|SetUpdateMaxNumberOfEarthquakesAtOnceTimerHandle _returnvalue)
      (CallFunction|UpdateMaxNumberOfEarthquakesatOnceValues) ; Call once to setup initial values.
    )))

(event Custom|StartEarthquakeSpawnCheckTimer
  ;[0.6472 0 1] Start a timer that attempts to spawn the next earthquake every interval seconds.
  (Networking|SwitchHasAuthority
    (:Authority
      (bind _returnvalue (Utilities|Time|SetTimerbyEvent (EventDispatchers|CreateEvent :!event Custom|TrySpawnEarthquake) (ScalableFloat|GetValueatLevel (Variables|Hotfix|GetEarthquakeSpawnCheckInterval)) true))
      (Variables|Default|SetEarthquakeSpawnCheckTimerHandle _returnvalue)
      (Fortnite|FortShippingLog self "StartEarthquakeSpawnCheckTimer: Invoked."))))

(event Custom|TrySpawnEarthquake
  ;[0.1094 0 0.2188] TrySpawnEarthquake: IF randomized earthquakes can be spawned, start an EQS to figure out the spawner to use.
  (Networking|SwitchHasAuthority
    (:Authority
      (bind _returnvalue (CallFunction|CanSpawnEarthquakes)) ; This function logs a reason if return is false.
      (if _returnvalue
        (Utilities|Time|ClearandInvalidateTimerbyHandle (Variables|Default|GetEarthquakeSpawnCheckTimerHandle)) ; Since we're looking to spawn an earthquake, \nstop the earthquake spawn check timer.
        (Variables|Default|GetEarthquakeSpawnPointQueryInstance
          (:then
            ;;[1 0 0] Warning: EarthquakeSpawnPointQueryInstance has a valid value when it shouldn't.
            (Fortnite|FortShippingLog self "TrySpawnEarthquake: EarthquakeSpawnPointQueryInstance has a valid value when it shouldn't." true))
          (:else
            (bind _returnvalue_2 (AI|EQS|RunEQSQuery "/Milkshake/Gameplay/EQS/EQS_Milkshake_RandomEarthquakeLocationFinder.EQS_Milkshake_RandomEarthquakeLocationFinder" self "AllMatching"))
            (Utilities|IsValid _returnvalue_2
              (:Is%20Valid
                (bind _output_get (Variables|Default|SetEarthquakeSpawnPointQueryInstance _returnvalue_2))
                (AI|EQS|BindEventtoOnQueryFinished _output_get (EventDispatchers|CreateEvent :!event Custom|EQSQueryComplete_RandomEarthquakeLocationFinder))
                (Fortnite|FortShippingLog self "TrySpawnEarthquake: Starting EQS to find spawner actor."))
              (:Is%20Not%20Valid
                ;;[1 0 0] Warning: Failed to create EQS Query.
                (Fortnite|FortShippingLog self "TrySpawnEarthquake: Failed to create EQS Query." true)
                (CallFunction|StartEarthquakeSpawnCheckTimer) ; Restart the timer that looks to spawn an earthquake so that we can try again.
              )))) ; Ensure that an EQS is not actually running.
      ))))

(event Custom|EQSQueryComplete_RandomEarthquakeLocationFinder
  (params (QueryInstance :type EQS%20Query%20Instance%20Object%20Reference) (QueryStatus :type EEnvQueryStatus%20Enum :default "Processing"))
  ;[0.01217 0.07812 0.06081] EQSQueryComplete_RandomEarthquakeLocationFinder: EQS is complete, a earthquake spawner location is found. Spawn an earthquake at that location actor using its settings.
  (Networking|SwitchHasAuthority
    (:Authority
      (Variables|Default|SetEarthquakeSpawnPointQueryInstance 0) ; Reset the query variable.
      (if (Utilities|Enum|Equal(Enum) QueryStatus "Success")
        (bind (_resultactors _returnvalue) (AI|EQS|GetQueryResultsasActors QueryInstance))
        (Fortnite|FortShippingLog self (Utilities|String|Append "EQSQueryComplete_RandomEarthquakeLocationFinder: Found actor count: " (Utilities|String|ToString(Integer) (Utilities|Array|Length _resultactors))) (not _returnvalue))
        (bind (_selectedactor _playercount) (CallFunction|GetActorClosesttoMostPlayers _resultactors))
        (bind _asbp_milkshake_random_earthquake_spawner (Utilities|Casting|CastToBP_Milkshake_RandomEarthquakeSpawner _selectedactor)
          (:then
            (Fortnite|FortShippingLog self (Utilities|String|Append "EQSQueryComplete_RandomEarthquakeLocationFinder: Player count  spawner " (Utilities|String|ToString(Integer) _playercount) " selected spawner name " (Class|BPMilkshakeRandomEarthquakeSpawner|GetSpawnerName _asbp_milkshake_random_earthquake_spawner)))
            (bind _earthquake (Class|BPMilkshakeRandomEarthquakeSpawner|SpawnEarthquake _asbp_milkshake_random_earthquake_spawner))
            (Utilities|IsValid _earthquake
              (:Is%20Valid
                (Utilities|Array|Add (Variables|Default|GetCurrentEarthquakes) _earthquake)
                (Game|AssignOnDestroyed _earthquake (EventDispatchers|CreateEvent :!event Custom|HandleEarthquakeComplete)) ; Bind to know when the earthquake finishes.
                (Math|Integer|IncrementInt (Variables|Default|GetTotalNumOfEarthquakesSpawned)
                  (:%20%20
                    (Fortnite|FortShippingLog self (Utilities|String|Append "EQSQueryComplete_RandomEarthquakeLocationFinder: Successfully spawned earthquake " (Utilities|GetDisplayName _earthquake) " at spawner name " (Class|BPMilkshakeRandomEarthquakeSpawner|GetSpawnerName _asbp_milkshake_random_earthquake_spawner)))
                    (CallFunction|StartEarthquakeSpawnCooldownTimer) ; An earthquake has been successfully spawned, start a cooldown timer.
                  )))
              (:Is%20Not%20Valid
                ;;[1 0 0] Warning: Failed to spawn earthquake.
                (Fortnite|FortShippingLog self (Utilities|String|Append "EQSQueryComplete_RandomEarthquakeLocationFinder: Failed spawn earthquake at spawner name:  " (Class|BPMilkshakeRandomEarthquakeSpawner|GetSpawnerName _asbp_milkshake_random_earthquake_spawner)) true)
                (CallFunction|StartEarthquakeSpawnCheckTimer) ; Restart the timer that looks to spawn an earthquake so that we can try again.
              )))
          (:CastFailed
            (Fortnite|FortShippingLog self "EQSQueryComplete_RandomEarthquakeLocationFinder: Failed to find Spawner close enough to players to use.")
            (CallFunction|StartEarthquakeSpawnCheckTimer) ; Restart the timer that looks to spawn an earthquake so that we can try again.
          ))
        (else
          ;;[0.5571 0.4162 1] EQS Query failed.
          (Fortnite|FortShippingLog self "EQSQueryComplete_RandomEarthquakeLocationFinder: Failed, not enough spawners available for use.")
          (CallFunction|StartEarthquakeSpawnCheckTimer) ; Restart the timer that looks to spawn an earthquake so that we can try again.
        )))))

(event Custom|StartEarthquakeSpawnCooldownTimer
  ;[0 0.05099 0.151] StartEarthquakeSpawnCooldownTimer: Start a cooldown timer before the next earthquake is attempted to be spawned.
  (Networking|SwitchHasAuthority
    (:Authority
      (bind _duration (CallFunction|GetEarthquakeSpawnCooldownDuration))
      (bind _returnvalue (Utilities|Time|SetTimerbyEvent (EventDispatchers|CreateEvent :!event Custom|HandleEarthquakeSpawnCooldownTimerComplete) _duration))
      (Variables|Default|SetEarthquakeSpawnCooldownTimer _returnvalue)
      (Fortnite|FortShippingLog self (Utilities|String|Append "StartEarthquakeSpawnCooldownTimer: Starting a cooldown timer for duration: " (Utilities|String|ToString(Float) _duration) " : at safe zone index : " (Utilities|String|ToString(Integer) (CallFunction|GetSafeZonePhase)))))))

(event Custom|HandleEarthquakeSpawnCooldownTimerComplete
  ;[0.1053 0.04053 0.1771] HandleEarthquakeSpawnCooldownTimerComplete: The cooldown timer is complete, look to spawn the next earthquake.
  (Networking|SwitchHasAuthority
    (:Authority
      (Fortnite|FortShippingLog self "HandleEarthquakeSpawnCooldownTimerComplete: Invoked.")
      (CallFunction|StartEarthquakeSpawnCheckTimer) ; Start the timer that attempts to spawn earthquakes.
      (CallFunction|TrySpawnEarthquake) ; Directly call TrySpawnEarthquake() cause the cooldown is complete.
    )))

(event Custom|RandomEarthquakesEventEnd
  ;[0.4896 0.1339 0.06535] RandomEarthquakeEventEnd: Once the calendar event has ended, clear out the timers responsible for spawning the earthquakes.
  (Networking|SwitchHasAuthority
    (:Authority
      (Fortnite|FortShippingLog self "RandomEarthquakesEventEnd: Invoked.")
      (Utilities|Time|ClearandInvalidateTimerbyHandle (Variables|Default|GetUpdateMaxNumberOfEarthquakesAtOnceTimerHandle))
      (Utilities|Time|ClearandInvalidateTimerbyHandle (Variables|Default|GetEarthquakeSpawnCheckTimerHandle))
      (Utilities|Time|ClearandInvalidateTimerbyHandle (Variables|Default|GetEarthquakeSpawnCooldownTimer))
      (bind _earthquakespawnpointqueryinstance (Variables|Default|GetEarthquakeSpawnPointQueryInstance)
        (:then
          (AI|EQS|UnbindEventfromOnQueryFinished _earthquakespawnpointqueryinstance (EventDispatchers|CreateEvent :!event Custom|EQSQueryComplete_RandomEarthquakeLocationFinder))
          (Fortnite|FortShippingLog self "RandomEarthquakesEventEnd: EQS was running."))
        (:else)) ; Unbind EQS query.
    )))

(event EventEndPlay
  (params (EndPlayReason :type EEndPlayReason%20Enum :default "Destroyed"))
  ;[0.2188 0.09129 0] EndPlay: Unregister any bound delegates.
  (|CallParentFunction EndPlayReason)
  (Networking|SwitchHasAuthority
    (:Authority
      (CallFunction|UnregisterfromActorRegistrySubsystemDelegates)
      (Utilities|FlowControl|Sequence
        (:then_0
          ;;[0.3542 0.1594 0.2805] Unbind from an active EQS query.
          (bind _earthquakespawnpointqueryinstance (Variables|Default|GetEarthquakeSpawnPointQueryInstance)
            (:then
              (AI|EQS|UnbindEventfromOnQueryFinished _earthquakespawnpointqueryinstance (EventDispatchers|CreateEvent :!event Custom|EQSQueryComplete_RandomEarthquakeLocationFinder)))
            (:else)))
        (:then_1
          ;;[0 0.02993 0.04688] Unbind OnDestroyed delegates from current earthquakes.
          (for _array_element (Variables|Default|GetCurrentEarthquakes)
            (Utilities|IsValid _array_element
              (:Is%20Valid
                (Game|BindEventtoOnDestroyed _array_element (EventDispatchers|CreateEvent :!event Custom|HandleEarthquakeComplete))))))))))

(event Custom|HandleEarthquakeComplete
  (params (DestroyedActor :type Actor%20Object%20Reference))
  ;[0 0.06962 0.1354] HandleEarthquakeComplete: Once the earthquake is complete, remove it from the current earthquakes list.
  (Networking|SwitchHasAuthority
    (:Authority
      (bind _asbp_milkshake_earthquake (Utilities|Casting|CastToBP_Milkshake_Earthquake DestroyedActor))
      (Utilities|Array|RemoveItem (Variables|Default|GetCurrentEarthquakes) _asbp_milkshake_earthquake)
      (Game|BindEventtoOnDestroyed _asbp_milkshake_earthquake (EventDispatchers|CreateEvent :!event Custom|HandleEarthquakeComplete)))))

(event Custom|MilkshakeSpawnEarthquake
  (params (EventName :type String))
  ;[0 0 0] Cheat ke * MilkshakeSpawnEarthquake (EventName)
  ;
  ; Spawns a single earthquake for the provided event name.
  (Networking|SwitchHasAuthority
    (:Authority
      (bind _outactors (Actor|GetAllActorsOfClass "/Milkshake/Gameplay/BP_Milkshake_RandomEarthquakeSpawner.BP_Milkshake_RandomEarthquakeSpawner_C"))
      (for _array_element _outactors
        (Utilities|IsValid _array_element
          (:Is%20Valid
            (if (Utilities|String|Equal,CaseInsensitive(String) (Class|BPMilkshakeRandomEarthquakeSpawner|GetSpawnerName _array_element) EventName)
              (Game|SpawnActorfromClass "/Milkshake/Gameplay/BP_Milkshake_Earthquake.BP_Milkshake_Earthquake_C" (Math|Conversions|ToTransform(Vector) (Transformation|GetActorLocation _array_element)) "Undefined" "OverrideRootScale" 0 0 (Class|BPMilkshakeRandomEarthquakeSpawner|GetEarthquakeInfo _array_element) 0 _array_element self 0))))))))

(event Custom|MilkshakeShowRandomSpawners
  ;[0 0 0] ke * MilkshakeShowRandomSpawners (Runs only for the player entering the cheat locally)
  ;
  ; Text render with the SpawerName value is spawned on top of the random spawner actors' location. Each spawner also gets an icon on the minimap.
  (Networking|CanExecuteCosmeticEvents
    (:True
      (if (not (Variables|Cheat|GetbCheat_SpawnersShowing))
        (Variables|Cheat|SetbCheat_SpawnersShowing true)
        (bind _returnvalue (Utilities|Time|SetTimerbyEvent (EventDispatchers|CreateEvent :!event Cheat_EventNameTextCompFaceCamera) 0.1 true false 0.2 0.2))
        (Variables|Cheat|SetCheat_TextCompFaceCameraTimerHandle _returnvalue)
        (bind _outactors (Actor|GetAllActorsOfClass "/Milkshake/Gameplay/BP_Milkshake_RandomEarthquakeSpawner.BP_Milkshake_RandomEarthquakeSpawner_C"))
        (for _array_element _outactors
          (Utilities|IsValid _array_element
            (:Is%20Valid
              (Utilities|FlowControl|Sequence
                (:then_0
                  (bind _returnvalue_2 (AddComponent|NODE_AddTextRenderComponent-0 :RelativeTransform_Location "0, 0, 0" :RelativeTransform_Rotation "0, 0, 0" :RelativeTransform_Scale "1.000000,1.000000,1.000000"))
                  (Utilities|IsValid _returnvalue_2
                    (:Is%20Valid
                      (Rendering|Components|TextRender|SetText _returnvalue_2 (Utilities|Text|FormatText "INVTEXT(\"{RowName}\")" (Utilities|Text|ToText(String) (Class|BPMilkshakeRandomEarthquakeSpawner|GetSpawnerName _array_element))))
                      (Utilities|Array|Add (Variables|Cheat|GetCheat_EventNameTextComponents) _returnvalue_2)
                      (Transformation|SetWorldLocation _returnvalue_2 (Utilities|Operators|Add (Transformation|GetActorLocation _array_element) (Math|Vector|MakeVector 0.0 0.0 (Variables|Cheat|GetCheat_TextComponentLocationOffsetZ))) false false))))
                (:then_1
                  (AddComponent|NODE_AddFortMiniMapComponent-1 :self _array_element))))))))))

(event Custom|MilkshakeHideRandomSpawners
  (bind _cheat_eventnametextcomponents (Variables|Cheat|GetCheat_EventNameTextComponents))
  ;[0 0 0] ke * MilkshakeHideRandomSpawners (Runs only for the player entering the cheat locally)
  ;
  ; Removes the text render with the SpawerName value on top of the random spawner actors' location along with the minimap icons.
  (Networking|CanExecuteCosmeticEvents
    (:True
      (Components|DestroyComponent _cheat_eventnametextcomponents)
      (Utilities|Array|Clear _cheat_eventnametextcomponents)
      (Variables|Cheat|SetbCheat_SpawnersShowing false)
      (Utilities|Time|ClearandInvalidateTimerbyHandle (Variables|Cheat|GetCheat_TextCompFaceCameraTimerHandle))
      (bind _outactors (Actor|GetAllActorsOfClass "/Milkshake/Gameplay/BP_Milkshake_RandomEarthquakeSpawner.BP_Milkshake_RandomEarthquakeSpawner_C"))
      (for _array_element _outactors
        (Components|RemoveArbitraryComponent _array_element "/Script/FortniteGame.FortMiniMapComponent")))))

(event Custom|MilkshakeSetEarthquakeShow
  (params (bShow :type Boolean))
  ;[0 0 0] ke * MilkshakeSetEarthquakeShow (Runs only for the player entering the cheat locally)
  ;
  ; Draws a sphere to represent the earthquake happening in the world.
  (Networking|CanExecuteCosmeticEvents
    (:True
      (Variables|Cheat|SetbCheat_EarthquakeShow bShow))))

(event Custom|MilkshakeSetRandomEarthquakeCalendarRatio
  (params (Value :type Float%20%28double-precision%29))
  ;[0 0 0] Cheat ke * MilkshakeSetRandomEarthquakeCalendarRatio (-1, 1)
  ;
  ; Cheat forces the calendar time ratio. Set to value < 0 to deactivate cheat.
  (Networking|SwitchHasAuthority
    (:Authority
      (Variables|Cheat|SetCheat_ForceCalendarTimeSpanRatio (Math|Float|Min(Float) Value 1.0)))))

(fn CanSpawnEarthquakes
  (bind _currentearthquakes (Variables|Default|GetCurrentEarthquakes))
  (bind _returnvalue (Utilities|Array|Length _currentearthquakes))
  (bind _returnvalue_2 (BattleRoyaleGamePhaseLogic|GetBattleRoyaleGamePhaseLogicComponent))
  (bind _returnvalue_3 (BattleRoyaleGamePhaseLogic|GetGamePhase _returnvalue_2))
  (bind _totalnumofearthquakesspawned (Variables|Default|GetTotalNumOfEarthquakesSpawned))
  (bind _numofrandomearthquakestospawn (Variables|Default|GetNumOfRandomEarthquakesToSpawn))
  ;[0.01636 0.07812 0.000399] Returns IF randomized earthquakes are allowed to be spawned at the time of call.
  (Utilities|FlowControl|Sequence
    (:then_0
      (Networking|SwitchHasAuthority
        (:Authority
          (Utilities|IsValid _returnvalue_2
            (:Is%20Valid
              ;;[0.3177 0.1299 0.0264] Earthquakes are only allowed once the safe zone commences.
              (if (Math|Boolean|ORBoolean (Utilities|Enum|Equal(Enum) _returnvalue_3 "SafeZones") (Utilities|Enum|Equal(Enum) _returnvalue_3 "EndGame") (Utilities|Enum|Equal(Enum) _returnvalue_3 "None"))
                ;;[0.08333 0.05944 0] Check to see if there's room to spawn more earthquakes at once.
                (bind _output_get (LocalVariables|SetCurrentMaxNumberOfEarthquakesAtOnce (CallFunction|GetCurrentMaxNumberOfEarthquakesatOnce)))
                (if (< _returnvalue _output_get)
                  ;;[0.1406 0.003609 0.06928] Check to see IF there's room to spawn more random earthquakes for the current match.
                  (if (< _totalnumofearthquakesspawned _numofrandomearthquakestospawn)
                    (return true)
                    (else
                      (Fortnite|FortShippingLog self (Utilities|String|Append "CanSpawnEarthquakes: False: TotalNumOfEarthquakesSpawned >= NumOfRandomEarthquakesToSpawn; " (Utilities|String|ToString(Integer) _totalnumofearthquakesspawned) ">=" (Utilities|String|ToString(Integer) _numofrandomearthquakestospawn)))))
                  (else
                    (Fortnite|FortShippingLog self (Utilities|String|Append "CanSpawnEarthquakes: False: CurrentEarthquakesNum >= CurrentMaxNumOfEarthquakeAtOnce; " (Utilities|String|ToString(Integer) _returnvalue) ">=" (Utilities|String|ToString(Integer) _output_get)))))
                (else
                  (Fortnite|FortShippingLog self (Utilities|String|Append "CanSpawnEarthquakes: False: Wrong game phase: " (Utilities|String|EnumtoString _returnvalue_3))))))
            (:Is%20Not%20Valid
              ;;[1 0 0] Warning: GSA is null.
              (Fortnite|FortShippingLog self "CanSpawnEarthquakes: Failed: GamePhaseLogic is null." true))))))
    (:then_1
      (return false))))

(fn Cheat_EventNameTextCompFaceCamera
  (bind _returnvalue (Fortnite|GetFirstLocalFortPlayerController true true))
  ;[0.05625 0.09679 0.125] Cheat_EventNameTextCompFaceCamera: Have the text components face the camera.
  (Networking|CanExecuteCosmeticEvents
    (:True
      (Utilities|IsValid _returnvalue
        (:Is%20Valid
          (bind _playercameramanager (Variables|/Script/Engine.PlayerController|GetPlayerCameraManager _returnvalue)
            (:then
              (for _array_element (Variables|Cheat|GetCheat_EventNameTextComponents)
                (Utilities|IsValid _array_element
                  (:Is%20Valid
                    (Transformation|SetWorldRotation _array_element (Math|Rotator|FindLookatRotation (Transformation|GetWorldLocation _array_element) (Camera|GetCameraLocation _playercameramanager)) false false)))))
            (:else)))))))

(fn Cheat_GetEarthquakeShow
  (return (Variables|Cheat|GetbCheat_EarthquakeShow)))

(fn GetActorClosestToMostPlayers
  (params (InActors :type Array%20of%20Actor%20Object%20References))
  ;[0.05801 0.01032 0.1354] Out of all the input actors, find the actor that has the most number of players near it.
  (Utilities|FlowControl|Sequence
    (:then_0
      (Networking|SwitchHasAuthority
        (:Authority
          (for _array_element InActors
            (Utilities|IsValid _array_element
              (:Is%20Valid
                (bind _numofplayers (CallFunction|GetNumberOfPlayersNearActor _array_element))
                (if (> _numofplayers (LocalVariables|GetHighestPlayerCount))
                  (LocalVariables|SetHighestPlayerCount _numofplayers)
                  (LocalVariables|SetReturnActor _array_element))))))))
    (:then_1
      (return (LocalVariables|GetReturnActor) (LocalVariables|GetHighestPlayerCount)))))

(fn GetCachedCalendarEventStatus
  ;[0.005208 0.001102 0.004292] Returns the calendar event status.
  (return (Variables|Default|GetCalendarStatus)))

(fn GetCalendarEventStatus
  (bind _cheat_forcecalendartimespanratio (Variables|Cheat|GetCheat_ForceCalendarTimeSpanRatio))
  ;[0 0.0407 0.08854] GetCalendarEventStatus: Takes debugging into consideration so that the feature can be tested by bypassing calendar checks.
  (Utilities|FlowControl|Sequence
    (:then_0
      (if (>= _cheat_forcecalendartimespanratio 0)
        (return true _cheat_forcecalendartimespanratio)
        (elif (Utilities|String|IsEmpty (Variables|Default|GetCalendarEvent))
          (return true 0.0)
          (else
            (bind _asfort_game_instance (Utilities|Casting|CastToFortGameInstance (Game|GetGameInstance)))
            (Utilities|IsValid (Class|FortGameInstance|GetEventManager _asfort_game_instance)
              (:Is%20Valid
                (bind (_biseventactive _timesincebegin _timeuntilend _timespanratio) (CalendarEvents|GetCalendarEventInformation (Variables|Default|GetCalendarEvent)))
                (return _biseventactive _timespanratio))) ; Make sure GetEventManger is valid before \nchecking calendar event info to prevent error spam.
          ))))
    (:then_1
      (return false 0.0))))

(fn GetCurrentMaxNumberOfEarthquakesAtOnce
  (bind _maxnumberofearthquakesatoncelist (Variables|Default|GetMaxNumberOfEarthquakesAtOnceList))
  ;[0.000581 0 0.07812] Returns the maximum number of earthquakes that can be spawned at once at the current safe zone phase.
  (bind _output_get (LocalVariables|SetSafeZonePhase (CallFunction|GetSafeZonePhase)))
  (if (Utilities|Array|IsValidIndex _maxnumberofearthquakesatoncelist _output_get)
    (return (Utilities|Array|Get(acopy) _maxnumberofearthquakesatoncelist _output_get))
    (else
      ;;[1 0 0] Warning: SafeZonePhase index not found in array.
      (Fortnite|FortShippingLog self (Utilities|String|Append "GetCurrentMaxNumberOfEarthquakesAtOnce: SafeZonePhase index is NOT found in MaxNumberOfEarthquakesAtOnceList: " (Utilities|String|ToString(Integer) _output_get)) true)
      (return 0))))

(fn GetEarthquakeSpawnCooldownDuration
  ;[0 0.01852 0.05208] Returns the earthquake spawn cooldown duration to use based on the safe zone phase index.
  (Utilities|FlowControl|Sequence
    (:then_0
      (Networking|SwitchHasAuthority
        (:Authority
          (return (Math|Random|RandomFloatinRange (ScalableFloat|GetValueatLevel (Variables|Hotfix|GetMinEarthquakeSpawnCooldownDuration) (Math|Conversions|ToFloat(Integer) (CallFunction|GetSafeZonePhase))) (ScalableFloat|GetValueatLevel (Variables|Hotfix|GetMaxEarthquakeSpawnCooldownDuration) (Math|Conversions|ToFloat(Integer) (CallFunction|GetSafeZonePhase))))))))
    (:then_1
      (return 0.0))))

(fn GetNumberOfPlayersNearActor
  (params (Actor :type Actor%20Object%20Reference))
  ;[0.04712 0.000761 0.06771] Return the number of players that are within relevancy range of the input actors.
  (Utilities|FlowControl|Sequence
    (:then_0
      (Networking|SwitchHasAuthority
        (:Authority
          (Utilities|IsValid Actor
            (:Is%20Valid
              (bind _outfortplayerpawns (Fortnite|GetFortPlayerPawnsInRange (Transformation|GetActorLocation Actor) (LocalVariables|GetRelevancyRange) true))
              (for _array_element _outfortplayerpawns
                (Utilities|IsValid _array_element
                  (:Is%20Valid
                    (if (Pawn|IsBotControlled _array_element)
                      (bind _asfort_athena_aibot_controller (Utilities|Casting|CastToFortAthenaAIBotController (Pawn|GetController _array_element)))
                      (if (Class|FortPlayerStateAthena|ControllerIsAnAthenaGameParticipant _asfort_athena_aibot_controller)
                        _)
                      (else
                        _))
                    (Math|Integer|IncrementInt (LocalVariables|GetReturnPlayerCount))))))))))
    (:then_1
      (return (LocalVariables|GetReturnPlayerCount)))))

(fn GetSafeZonePhase
  (bind _returnvalue (SafeZone|GetSafeZoneIndicator))
  ;[0 0.07639 0.09896] Returns the current safe zone phase.
  (Utilities|IsValid _returnvalue
    (:Is%20Valid
      (return (SafeZone|GetCurrentPhaseIndex _returnvalue)))
    (:Is%20Not%20Valid
      (return 0))))

(fn HandleActorRegistrySubsystemRegisterActor
  (params (GameplayTag :type Gameplay%20Tag%20Structure%20%28by%20ref%29) (RegisteredActor :type Actor%20Object%20Reference))
  ;[0 0.0562 0.09896] Update the earthquake manager reference for every random earthquake spawner in the world.
  (Networking|SwitchHasAuthority
    (:Authority
      (bind _asbp_milkshake_random_earthquake_spawner (Utilities|Casting|CastToBP_Milkshake_RandomEarthquakeSpawner RegisteredActor))
      (Class|BPMilkshakeRandomEarthquakeSpawner|SetRandomEarthquakeManager _asbp_milkshake_random_earthquake_spawner self))))

(fn RegisterToActorRegistrySubsystemDelegates
  (bind _returnvalue (WorldSubsystems|GetFortGlobalActorRegistrySubsystem))
  (bind _randomearthquakespawnerregistertag (Variables|Default|GetRandomEarthquakeSpawnerRegisterTag))
  ;[0.0855 0 0.1146] Setup delegates to find all the earthquake spawners in the world.
  ; The earthquake spawners are given the reference to this (manager) actor blueprint.
  (Networking|SwitchHasAuthority
    (:Authority
      (Utilities|IsValid _returnvalue
        (:Is%20Valid
          (bind _returnvalue_2 (Class|FortGlobalActorRegistrySubsystem|GetActorRegistryDataforGameplayTag _returnvalue _randomearthquakespawnerregistertag))
          (Utilities|IsValid _returnvalue_2
            (:Is%20Valid
              (EventDispatchers|BindEventtoOnActorRegistered _returnvalue_2 (EventDispatchers|CreateEvent :!event HandleActorRegistrySubsystemRegisterActor))
              (for _array_element (Class|ActorRegisterDataGameplayTagBased|GetRegisteredActors _returnvalue_2)
                (Utilities|IsValid _array_element
                  (:Is%20Valid
                    (CallFunction|HandleActorRegistrySubsystemRegisterActor _randomearthquakespawnerregistertag _array_element)))))))
        (:Is%20Not%20Valid
          (Fortnite|FortShippingLog self "RegisterToActorRegistrySubsystemDelegates: Failed to get GlobalActorRegistry." true))))))

(fn UnregisterFromActorRegistrySubsystemDelegates
  (bind _returnvalue (WorldSubsystems|GetFortGlobalActorRegistrySubsystem))
  ;[0.1975 0.03239 0.2135] Unregister from the actor registry subsystem.
  (Networking|SwitchHasAuthority
    (:Authority
      (Utilities|IsValid _returnvalue
        (:Is%20Valid
          (bind _returnvalue_2 (Class|FortGlobalActorRegistrySubsystem|GetActorRegistryDataforGameplayTag _returnvalue (Variables|Default|GetRandomEarthquakeSpawnerRegisterTag)))
          (Utilities|IsValid _returnvalue_2
            (:Is%20Valid
              (EventDispatchers|UnbindEventfromOnActorRegistered _returnvalue_2 (EventDispatchers|CreateEvent :!event HandleActorRegistrySubsystemRegisterActor)))))
        (:Is%20Not%20Valid
          (Fortnite|FortShippingLog self "UnregisterFromActorRegistrySubsystemDelegates: Failed to get GlobalActorRegistry." true))))))

(fn UpdateCalendarEventStatus
  ;[0 0.03738 0.1302] Update the calendar event status responsible for the randomized earthquakes.
  (Networking|SwitchHasAuthority
    (:Authority
      (bind (_biseventactive _timespanratio) (CallFunction|GetCalendarEventStatus))
      (Utilities|Struct|SetmembersinMilkshakeCalendarEventInfoStruct (Variables|Default|GetCalendarStatus) _timespanratio)
      (if (!= (Utilities|Struct|BreakMilkshakeCalendarEventInfoStruct (Variables|Default|GetCalendarStatus)) _biseventactive)
        (Utilities|Struct|SetmembersinMilkshakeCalendarEventInfoStruct (Variables|Default|GetCalendarStatus) _biseventactive)
        (if _biseventactive
          (CallFunction|RandomEarthquakesEventStart)
          (else
            (CallFunction|RandomEarthquakesEventEnd)))))))

(fn UpdateMaxNumberOfEarthquakesAtOnceValues
  ;[0.1506 0.04704 0.1615] Updates the number of earthquakes that are allowed to spawn at once per safe zone index.
  (Utilities|Array|Clear (Variables|Default|GetMaxNumberOfEarthquakesAtOnceList))
  (for _index (range (+ (LocalVariables|GetHighestSafeZoneIndex) 1))
    (bind _output_get (LocalVariables|SetValue (Math|Random|RandomIntegerinRange (ScalableFloat|GetValueasInteger (Variables|Hotfix|GetMinNumberOfEarthquakesAtOnce) (Math|Conversions|ToFloat(Integer) _index)) (ScalableFloat|GetValueasInteger (Variables|Hotfix|GetMaxNumberOfEarthquakesAtOnce) (Math|Conversions|ToFloat(Integer) _index)))))
    (Utilities|Array|SetArrayElem (Variables|Default|GetMaxNumberOfEarthquakesAtOnceList) _index _output_get true)
    (LocalVariables|SetDebugString (Utilities|String|Append (LocalVariables|GetDebugString) ", " (Utilities|String|ToString(Integer) _output_get))))
  (Fortnite|FortShippingLog self (Utilities|String|Append "UpdateMaxNumberOfEarthquakesAtOnceValues: " (LocalVariables|GetDebugString))))

(fn ConstructionScript)
