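; Milkshake random earthquake spawner: a placed actor that spawns a BP_Milkshake_Earthquake at its location on the server, using a datatable row keyed by its spawner name.
; BeginPlay rolls a maximum spawn count, and each destroyed quake starts a random cooldown timer.
; CanSpawnEarthquake, used by an EQS generator, gates on cooldown, active quake, count, calendar event, safe-zone phase and playlist tags; the random earthquake manager calls SpawnEarthquake.
; Editor functions draw and clear debug radius actors.
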(event EventBeginPlay
  (bind (_x _y) (Math|Vector2D|BreakVector2D (Utilities|Struct|BreakMilkshakeEarthquakeInfoStruct (CallFunction|GetEarthquakeInfo))))
  ;[0 0.151 0] BeginPlay: Set the maximum number of earthquakes that can spawn at this location.
  (|CallParentFunction)
  (Networking|SwitchHasAuthority
    (:Authority
      (bind _output_get (Variables|Default|SetMaxNumOfEarthquakes (Math|Float|Truncate (Math|Random|RandomFloatinRange _x _y))))
      (Fortnite|FortShippingLog self (Utilities|String|Append "BeginPlay: MaxNumOfEarthquakes  for spawner " (CallFunction|GetSpawnerName) " is " (Utilities|String|ToString(Integer) _output_get))))))

(event Custom|HandleEarthquakeDestroyed
  (params (DestroyedActor :type Actor%20Object%20Reference))
  (bind (_x _y) (Math|Vector2D|BreakVector2D (Utilities|Struct|BreakMilkshakeEarthquakeInfoStruct (CallFunction|GetEarthquakeInfo))))
  ;[0.0625 0 0.053] HandleEarthquakeDestroyed: Called when the earthquake spawned for this actor is destroyed.
  (Networking|SwitchHasAuthority
    (:Authority
      (Game|UnbindEventfromOnDestroyed (Variables|Default|GetCurrentEarthquake) (EventDispatchers|CreateEvent :!event Custom|HandleEarthquakeDestroyed))
      (Variables|Default|SetCurrentEarthquake 0)
      (bind _returnvalue (Utilities|Time|SetTimerbyEvent (EventDispatchers|CreateEvent :!event Custom|HandleSpawnerCooldownTimerComplete) (Math|Random|RandomFloatinRange _x _y)))
      (bind _output_get (Variables|Default|SetSpawnerCooldownTimerHandle _returnvalue))
      (Fortnite|FortShippingLog self (Utilities|String|Append "HandleEarthquakeDestroyed: Earthquake destroyed for spawner " (CallFunction|GetSpawnerName) ". Starting cooldown timer for duration: " (Utilities|String|ToString(Float) (Utilities|Time|GetTimerRemainingTimebyHandle _output_get)))))))

(event EventEndPlay
  (params (EndPlayReason :type EEndPlayReason%20Enum :default "Destroyed"))
  ;[0.1146 0.0442 0] EndPlay: Cleanup bound delegates.
  (|CallParentFunction EndPlayReason)
  (Networking|SwitchHasAuthority
    (:Authority
      (bind _currentearthquake (Variables|Default|GetCurrentEarthquake)
        (:then
          (Game|UnbindEventfromOnDestroyed _currentearthquake (EventDispatchers|CreateEvent :!event Custom|HandleEarthquakeDestroyed)))
        (:else)))))

(event Custom|HandleSpawnerCooldownTimerComplete
  ;[0 0.0078 0.09896] HandleSpawnerCooldownTimerComplete: Once the cooldown is complete, invalidate the cooldown timer handle.
  (Networking|SwitchHasAuthority
    (:Authority
      (Utilities|Time|ClearandInvalidateTimerbyHandle (Variables|Default|GetSpawnerCooldownTimerHandle))
      (Fortnite|FortShippingLog self (Utilities|String|Append "HandleSpawnerCooldownTimerComplete: Cooldown complete for spawner " (CallFunction|GetSpawnerName))))))

(fn CanSpawnEarthquake
  (bind _returnvalue (SafeZone|GetSafeZoneIndicator))
  ;[0.07812 0 0.06536] CanSpawnEarthquake: Returns if this actor can currently spawn an earthquake.
  ; Called from EnvQueryGenerator_Milkshake_ValidEarthquakeSpawners.
  (Utilities|FlowControl|Sequence
    (:then_0
      (Networking|SwitchHasAuthority
        (:Authority
          ;;[1 1 1] Is not in cooldown?
          (if (not (Utilities|Time|IsTimerActivebyHandle (Variables|Default|GetSpawnerCooldownTimerHandle)))
            ;;[1 1 1] Does not currently have an earthquake happening here.
            (Variables|Default|GetCurrentEarthquake
              (:then)
              (:else
                ;;[1 1 1] Has not reached max earthquake spawned amount.
                (if (< (Variables|Default|GetNumOfSpawnedEarthquakes) (Variables|Default|GetMaxNumOfEarthquakes))
                  ;;[1 1 1] Is currently spawner currently enabled? Is calendar event active and within specified time span ratio?
                  (bind _randomearthquakemanager (Variables|Default|GetRandomEarthquakeManager)
                    (:then
                      (bind _output_get (LocalVariables|SetEarthquakeInfo (CallFunction|GetEarthquakeInfo)))
                      (if (Math|Boolean|ANDBoolean (ScalableFloat|GetValueasBool (Utilities|Struct|BreakMilkshakeEarthquakeInfoStruct _output_get)) (Utilities|Struct|BreakMilkshakeCalendarEventInfoStruct (Class|BPMilkshakeRandomEarthquakeManager|GetCachedCalendarEventStatus _randomearthquakemanager)) (Math|Float|InRange(Float) (.TimeSpanRatio_4_41A67F774E2FE74E26C7968C3358303B (Utilities|Struct|BreakMilkshakeCalendarEventInfoStruct (Class|BPMilkshakeRandomEarthquakeManager|GetCachedCalendarEventStatus _randomearthquakemanager))) (Math|Vector2D|BreakVector2D (Utilities|Struct|BreakMilkshakeEarthquakeInfoStruct _output_get)) (.Y (Math|Vector2D|BreakVector2D (Utilities|Struct|BreakMilkshakeEarthquakeInfoStruct _output_get)))))
                        ;;[1 1 1] Is in a valid safe zone phase?
                        (Utilities|IsValid _returnvalue
                          (:Is%20Valid
                            (if (and (Utilities|Array|IsValidIndex (.SafeZonePhase_106_F9A528FA4C9CFAF4C009FD820928814E (Utilities|Struct|BreakMilkshakeEarthquakeInfoStruct _output_get)) (SafeZone|GetCurrentPhaseIndex _returnvalue)) (Utilities|Array|Get(acopy) (.SafeZonePhase_106_F9A528FA4C9CFAF4C009FD820928814E (Utilities|Struct|BreakMilkshakeEarthquakeInfoStruct _output_get)) (SafeZone|GetCurrentPhaseIndex _returnvalue)))
                              ;;[1 1 1] Does current playlist allow this earthquake?
                              (bind _ballow (CallFunction|DoesCurrentPlaylistAllowEarthquake (.RequiredPlaylistTags_113_33DFD1284D4FDA8A1D11D8804B982C77 (Utilities|Struct|BreakMilkshakeEarthquakeInfoStruct _output_get))))
                              (return _ballow))))))
                    (:else)))))))))
    (:then_1
      (return false))))

(fn DoesCurrentPlaylistAllowEarthquake
  (params (RequiredPlaylistTags :type Gameplay%20Tag%20Container%20Structure))
  (bind _returnvalue (Fortnite|Athena|GameState|GetGameStateAthena))
  ;[0.03454 0 0.1719] Returns if the loaded playlist has any of the required playlist tags.
  (Utilities|FlowControl|Sequence
    (:then_0
      (Utilities|IsValid _returnvalue
        (:Is%20Valid
          (bind _output_get (LocalVariables|SetCurrentGameState _returnvalue))
          (if (Utilities|GameplayTagContainerNotEmpty RequiredPlaylistTags)
            (return (GameplayTags|HasAnyTags (Class|FortGameStateAthena|GetAthenaPlaylistContextTags _output_get) RequiredPlaylistTags true) (Utilities|String|ToString(Name) (Playlist|GetCurrentPlaylistName _output_get)))
            (else
              (return true (Utilities|String|ToString(Name) (Playlist|GetCurrentPlaylistName _output_get))))))))
    (:then_1
      (return false ""))))

(fn Editor_DestroyAllDebugRandomEarthquake
  ;[0.08841 0.02564 0.1094] Editor_DestroyAllDebugRandomEarthquake: Destroy all the debug earthquake spheres representing the random earthquakes.
  (bind _outactors (Actor|GetAllActorsOfClass "/Milkshake/Gameplay/BP_Milkshake_EarthquakeDebug.BP_Milkshake_EarthquakeDebug_C"))
  (for _array_element _outactors
    (Utilities|IsValid _array_element
      (:Is%20Valid
        (Utilities|IsValid (Actor|GetAttachParentActor _array_element)
          (:Is%20Valid
            (Utilities|Casting|CastToBP_Milkshake_RandomEarthquakeSpawner (Actor|GetAttachParentActor _array_element))
            (Actor|DestroyActor _array_element)))))))

(fn Editor_DestroyDebugRandomEarthquake
  ;[0.2917 0.005866 0.2141] Editor_DestroyDebugRandomEarthquake: Destroy the debug sphere on this actor.
  (for _array_element (Actor|GetAttachedActors true)
    (bind _asbp_milkshake_earthquake_debug (Utilities|Casting|CastToBP_Milkshake_EarthquakeDebug _array_element))
    (Actor|DestroyActor _asbp_milkshake_earthquake_debug)))

(fn Editor_DrawAllDebugRandomEarthquake
  ;[0 0.06269 0.2448] Editor_DrawAllDebugRandomEarthquake: Spawn debug actors in the world that represent the earthquake (radius) based on the values entered in the random earthquake datatable.
  ; Warning: These debug actors are ACTUALLY spawned into the level so it is important to clear them up using the Editor_DestroyAllDebugEarthquakeRadius before saving the level!
  ; Note: IF DebugEarthquakeDataTableRowNames is empty, all the earthquakes from the EarthquakeInfoDatatable is debugged.
  (bind _outactors (Actor|GetAllActorsOfClass "/Milkshake/Gameplay/BP_Milkshake_RandomEarthquakeSpawner.BP_Milkshake_RandomEarthquakeSpawner_C"))
  (for _array_element _outactors
    (Utilities|IsValid _array_element
      (:Is%20Valid
        (Editor-Debug|EditorDrawDebugRandomEarthquake _array_element)))))

(fn Editor_DrawDebugRandomEarthquake
  (bind (_x _y) (Math|Vector2D|BreakVector2D (Utilities|Struct|BreakMilkshakeEarthquakeInfoStruct (CallFunction|GetEarthquakeInfo))))
  ;[0 0.06269 0.2448] Editor_DrawDebugRandomEarthquake: Spawn debug actors in the world that represent the earthquake (radius) based on the values entered in the random earthquake datatable.
  ; Warning: These debug actors are ACTUALLY spawned into the level so it is important to clear them up using the Editor_DestroyAllDebugEarthquakeRadius before saving the level!
  ; Note: IF DebugEarthquakeDataTableRowNames is empty, all the earthquakes from the EarthquakeInfoDatatable is debugged.
  (Editor-Debug|EditorDestroyDebugRandomEarthquake)
  (bind _returnvalue (Game|SpawnActorfromClass "/Milkshake/Gameplay/BP_Milkshake_EarthquakeDebug.BP_Milkshake_EarthquakeDebug_C" (Math|Transform|MakeTransform (Transformation|GetActorLocation)) "AlwaysSpawn" "OverrideRootScale" 0 (Utilities|Text|ToText(String) (CallFunction|GetSpawnerName)) (Transformation|GetActorLocation) (Math|Float|Max(Float) _x _y) (Transformation|GetActorLocation) 0))
  (Utilities|IsValid _returnvalue
    (:Is%20Valid
      (Transformation|AttachActorToActor _returnvalue self "None" "KeepWorld" "KeepWorld" "KeepWorld" false) ; Attach it to the earthquake so its easier to parse/understand\n when looking at the world outliner.
    )))

(fn GetEarthquakeInfo
  ;[0.07696 0 0.1302] Returns the earthquake info from the input datatable.
  (Utilities|FlowControl|Sequence
    (:then_0
      (bind _randomearthquakesinfodatatable (Variables|Default|GetRandomEarthquakesInfoDatatable)
        (:then
          (bind _returnvalue (Utilities|GetDataTableRow _randomearthquakesinfodatatable (Variables|Default|GetSpawnerName))
            (:then
              (return _returnvalue))
            (:RowNotFound
              ;;[1 0 0] Warning: Row not found in datatable.
              (Fortnite|FortShippingLog self (Utilities|String|Append "GetEarthquakeInfo: Datatable row not found: " (CallFunction|GetSpawnerName) "; Datatable: " (Utilities|GetDisplayName _randomearthquakesinfodatatable)) true))))
        (:else
          ;;[1 0 0] Warning: Datatable is null for spawner.
          (Fortnite|FortShippingLog self (Utilities|String|Append "GetEarthquakeInfo: Datatable is null for spawner: " (CallFunction|GetSpawnerName)) true))))
    (:then_1
      (return 0))))

(fn GetSpawnerName
  (bind _spawnername (Variables|Default|GetSpawnerName))
  ;[0.1927 0.06341 0.1232] Returns the spawner name used as the datatable row name.
  (if (== (LocalVariables|GetEmptyName) _spawnername)
    (return (Utilities|GetDisplayName self))
    (else
      (return (Utilities|String|ToString(Name) _spawnername)))))

(fn IsDebugEnabled
  ;[0.1614 0.00213 0.1719] Returns true if bDebug is set to true and the game is running in editor.
  (return (and (Variables|Editor-Debug|GetbDebug) (Editor|InEditororPIE)) (and (and (Variables|Editor-Debug|GetbDebug) (Editor|InEditororPIE)) (Variables|Editor-Debug|GetbDebugDrawRuntimeOutlines))))

(fn SetRandomEarthquakeManager
  (params (InManager :type BP%20Milkshake%20Random%20Earthquake%20Manager%20Object%20Reference))
  ;[0.05712 0 0.09896] Set the random earthquake manager variable.
  ; Called from BP_Milkshake_RandomEarthquakeManager.
  (Networking|SwitchHasAuthority
    (:Authority
      (Variables|Default|SetRandomEarthquakeManager InManager))))

(fn SpawnEarthquake
  ;[0.1094 0 0.08384] SpawnEarthquake: Spawns the earthquake with the specified earthquake info.
  ; Called from BP_Milkshake_RandomEarthquakeManager.
  (Utilities|FlowControl|Sequence
    (:then_0
      (Networking|SwitchHasAuthority
        (:Authority
          (Variables|Default|GetCurrentEarthquake
            (:then
              ;;[1 0 0] Warning: CurrentEarthquake has a valid value.
              (Fortnite|FortShippingLog self (Utilities|String|Append "SpawnEarthquake: CurrentEarthquake has a valid value. SpawnerName: " (CallFunction|GetSpawnerName)) true))
            (:else
              (bind _returnvalue (Game|SpawnActorfromClass "/Milkshake/Gameplay/BP_Milkshake_Earthquake.BP_Milkshake_Earthquake_C" (Math|Conversions|ToTransform(Vector) (Transformation|GetActorLocation)) "Undefined" "OverrideRootScale" 0 0 (CallFunction|GetEarthquakeInfo) 0 self (Variables|Default|GetRandomEarthquakeManager) 0))
              (Utilities|IsValid _returnvalue
                (:Is%20Valid
                  (Math|Integer|IncrementInt (Variables|Default|GetNumOfSpawnedEarthquakes)
                    (:%20%20
                      (bind _output_get (Variables|Default|SetCurrentEarthquake _returnvalue))
                      (Game|BindEventtoOnDestroyed _output_get (EventDispatchers|CreateEvent :!event HandleEarthquakeDestroyed)))))
                (:Is%20Not%20Valid
                  ;;[1 0 0] Warning: CurrentEarthquake has a valid value.
                  (Fortnite|FortShippingLog self (Utilities|String|Append "SpawnEarthquake: Failed to spawn earthquake. SpawnerName: " (CallFunction|GetSpawnerName)) true)))))
          (return (Variables|Default|GetCurrentEarthquake)))))
    (:then_1
      (return 0))))

(fn ConstructionScript)
