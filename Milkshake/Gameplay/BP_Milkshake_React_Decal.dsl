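; Milkshake earthquake reaction actor that forms a cracked-ground decal when an earthquake triggers its StartReact.
; BeginPlay gives the decal a dynamic material instance on cosmetic clients.
; StartReact runs once: after a delay, it shows the decal, sets material Time and a random Duration to animate cracking, and activates quake particles sized from the scaled decal.
; It also plays a crack sound, then a delayed world camera shake.
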(event EventBeginPlay
  ;[0.02595 1 0] BeginPlay: Set up dynamic material instance for the decals.
  (|CallParentFunction)
  (Networking|CanExecuteCosmeticEvents
    (:True
      (CallFunction|SetupDecalDynamicMaterialInstance))))

(event EventStartReact
  (params (OverlappingEarthquake :type BP%20Milkshake%20Earthquake%20Object%20Reference))
  (bind (_returnvalue_x _returnvalue_y) (Math|Vector|BreakVector (* (Transformation|GetActorScale3D) (Variables|/Script/Engine.DecalComponent|GetDecalSize (Variables|Default|GetCrackedGroundDecal)))))
  (bind _decaldynamicmaterialinstance (Variables|Default|GetDecalDynamicMaterialInstance))
  (bind (_crackformationminmaxduration_x _crackformationminmaxduration_y) (Math|Vector2D|BreakVector2D (Variables|GetCrackFormationMinMaxDuration_X)))
  (bind _ns_milkshake_quake_decal (Variables|Default|GetNS_Milkshake_Quake_Decal))
  ;[0 0.4583 0.2957] StartReact: Start drawing the cracks on the ground.
  (|CallParentFunction OverlappingEarthquake)
  (Networking|CanExecuteCosmeticEvents
    (:True
      (Utilities|FlowControl|DoOnce false
        (:Completed
          (Fortnite|FortShippingLog self (Utilities|String|Append "StartReact: Invoked with OverlappingEarthquake as " (Utilities|GetDisplayName OverlappingEarthquake)))
          (bind _output_get (Variables|Default|SetOverlappingEarthquake OverlappingEarthquake))
          (Utilities|FlowControl|Delay (Variables|Default|GetCrackFormationStartDelay))
          (Utilities|IsValid _output_get
            (:Is%20Valid
              (Rendering|SetVisibility (Variables|Default|GetCrackedGroundDecal) true false)
              (Rendering|Material|SetScalarParameterValue _decaldynamicmaterialinstance "Time" (Utilities|Time|GetGameTimeinSeconds) :!class /Script/Engine.MaterialInstanceDynamic)
              (Rendering|Material|SetScalarParameterValue _decaldynamicmaterialinstance "Duration" (Math|Random|RandomFloatinRange _crackformationminmaxduration_x _crackformationminmaxduration_y) :!class /Script/Engine.MaterialInstanceDynamic)
              (Components|Activation|SetActive _ns_milkshake_quake_decal true false)
              (Effects|Components|ParticleSystem|SetFloatParameter _ns_milkshake_quake_decal "Radius" _returnvalue_y)
              (Audio|SpawnSoundatLocation (Variables|Sound|GetCracked%20Ground%20SFX) (Transformation|GetWorldLocation (Variables|Default|GetCrackedGroundDecal)))
              (Utilities|FlowControl|Delay (Math|Random|RandomFloatinRange 0.5 2.0))
              (Camera|PlayWorldCameraShake "/FNGameplayCues/Effects/Fort_Effects/BluePrints/CameraShakes/B_Shudder_Vertical_Jolt_CameraShake_Long.B_Shudder_Vertical_Jolt_CameraShake_Long_C" (Transformation|GetActorLocation) 5e+02 2.5e+03))))) ; There's no need to run the decal formation more than once.
    )))

(fn SetupDecalDynamicMaterialInstance
  ;[0 0.9058 1] SetupDecalDynamicMaterialInstance: Sets up the dynamic material instance for the decal.
  (bind _returnvalue (Rendering|Material|CreateDynamicMaterialInstance (Variables|/Script/Engine.DecalComponent|GetDecalMaterial (Variables|Default|GetCrackedGroundDecal)) "DecalDynamicMaterialInstance" "None" :!class /Script/Engine.KismetMaterialLibrary))
  (Utilities|IsValid _returnvalue
    (:Is%20Valid
      (Rendering|Components|Decal|SetDecalMaterial (Variables|Default|GetCrackedGroundDecal) _returnvalue)
      (Variables|Default|SetDecalDynamicMaterialInstance _returnvalue))))

(fn ConstructionScript
  (|CallParentFunction))
