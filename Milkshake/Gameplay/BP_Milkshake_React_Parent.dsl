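; Base actor for objects that react to Milkshake earthquakes, destroyed by the server at BeginPlay if its scalable-float enable flag is off.
; Earthquakes report overlap on a timer, and the server tracks one earthquake, clearing it when overlap ends.
; That variable's RepNotify flushes net dormancy on the server and calls StartReact or EndReact on server and clients.
; Both events only log here, as hooks for child Blueprints to override.
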(event EventBeginPlay
  ;[0 1 0] BeginPlay: Destroy the actor IF it isn't enabled.
  (|CallParentFunction)
  (Networking|SwitchHasAuthority
    (:Authority
      (if (not (ScalableFloat|GetValueasBool (Variables|Default|GetbEnabled)))
        (Networking|FlushNetDormancy)
        (Actor|DestroyActor)))))

(event Custom|StartReact
  (params (OverlappingEarthquake :type BP%20Milkshake%20Earthquake%20Object%20Reference))
  ;[0 0 0] Overridden in children for custom behaviors: Do NOT delete.
  (Fortnite|FortShippingLog self (Utilities|String|Append "StartReact: Invoked with OverlappingEarthquake as " (Utilities|GetDisplayName OverlappingEarthquake))))

(event Custom|EndReact
  ;[0 0 0] Overridden in children for custom behaviors: Do NOT delete.
  (Fortnite|FortShippingLog self "EndReact: Invoked"))

(fn OnRep_CurrentOverlapEarthquakeActor
  ;[0 0.09364 0.1719] OnRep_CurrentOverlapEarthquakeActor: Run the respective event based on the value of the overlap earthquake actor variable.
  ; Called on both server and client when the value of the CurrentOverlapEarthquakeActor is updated.
  (Utilities|FlowControl|Sequence
    (:then_0
      ;;[1 1 1] On the server, flush net dormancy so that this actor is able to start replicating.
      (Networking|SwitchHasAuthority
        (:Authority
          (Networking|FlushNetDormancy))))
    (:then_1
      (bind _currentoverlapearthquakeactor (Variables|Default|GetCurrentOverlapEarthquakeActor)
        (:then
          (CallFunction|StartReact _currentoverlapearthquakeactor))
        (:else
          (CallFunction|EndReact))))))

(fn UpdateEarthquakeOverlappingStatus
  (params (IsOverlapping :type Boolean) (OverlappingEarthquakeActor :type BP%20Milkshake%20Earthquake%20Object%20Reference))
  ;[0.1971 0.2344 0.02203] UpdateEarthquakeOverlappingStatus: Update the overlapping status and variable of the input earthquake actor.
  ; This is called directly from the earthquake blueprint on a timer with the latest status information.
  (Networking|SwitchHasAuthority
    (:Authority
      (Variables|Default|GetCurrentOverlapEarthquakeActor
        (:then
          (if (== (Variables|GetOverlappingEarthquakeActor) (Variables|Default|GetCurrentOverlapEarthquakeActor))
            _))
        (:else
          (if (Variables|GetIsOverlapping)
            (Variables|Default|SetCurrentOverlapEarthquakeActor (Variables|GetOverlappingEarthquakeActor)))))
      (if (not (== (Variables|Default|GetCurrentOverlapEarthquakeStatus) (Variables|GetIsOverlapping)))
        (bind _output_get (Variables|Default|SetCurrentOverlapEarthquakeStatus (Variables|GetIsOverlapping)))
        (if (not _output_get)
          (Variables|Default|SetCurrentOverlapEarthquakeActor 0))))))

(fn ConstructionScript)
