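; Spline actor that schedules and hosts Milkshake earthquake events defined in a data table, driven by a calendar event.
; On the server, once playlist data is ready and hotfix flags allow, it polls calendar status on a timer and adds a manager component per row that passes hotfix and playlist-tag checks.
; Rows map calendar-ratio windows to normalized spline positions.
; Cheats force the ratio, spawn quakes or visualize the spline; editor helpers spawn debug actors.
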(event EventBeginPlay
  (bind _bmilkshakeenabled (Variables|Default|GetbMilkshakeEnabled))
  (bind _returnvalue (ScalableFloat|GetValueatLevel _bmilkshakeenabled))
  (bind _bmilkshakesplineenabled (Variables|Default|GetbMilkshakeSplineEnabled))
  (bind _returnvalue_2 (ScalableFloat|GetValueatLevel _bmilkshakesplineenabled))
  (bind _returnvalue_3 (> _returnvalue_2 0))
  (bind _returnvalue_4 (> _returnvalue 0))
  (bind _returnvalue_5 (and _returnvalue_4 _returnvalue_3))
  ;[0.02236 0.3906 0] BeginPlay: On the server, setup delegate to watch for earthquake calendar events.
  (|CallParentFunction)
  (Networking|SwitchHasAuthority
    (:Authority
      (Class|FortAsyncActionCurrentPlaylistReady|PlaylistDataReadyAsync
        (:OnReady
          (Fortnite|FortShippingLog self (Utilities|String|Append "BeginPlay: Spawned. Hotfix: " (Utilities|String|ToString(Boolean) _returnvalue_5) "(bMilkshakeEnabled: " (Utilities|String|ToString(Float) _returnvalue) ", bMilkshakeSplineEnabled: " (Utilities|String|ToString(Float) _returnvalue_2) ")"))
          (if _returnvalue_5
            ;;[0 0.1406 0.07617] Start a timer to track the calendar event status.
            (Utilities|Time|SetTimerbyEvent (EventDispatchers|CreateEvent :!event UpdateCalendarEventStatus) (ScalableFloat|GetValueatLevel (Variables|Default|GetCalendarEventStatusUpdateInterval)) true false 0.5 0.2)
            (CallFunction|UpdateCalendarEventStatus) ; Run it once on BeginPlay() for initial values.
            (bind _earthquakeinfodatatable (Variables|Default|GetEarthquakeInfoDatatable)
              (:then
                (bind _outrownames (DataTable|GetDataTableRowNames _earthquakeinfodatatable))
                (for _array_element _outrownames
                  (CallFunction|CreateEarthquakeComponent (Utilities|String|ToString(Name) _array_element))))
              (:else
                ;;[1 0 0] Warning: Earthquake Info Datatable is null.
                (Fortnite|FortShippingLog self "BeginPlay: EarthquakeInfoDatatable is null." true)))))) ; Wait until the playlist data is ready before checking hotfix values
    )))

(event EventEndPlay
  (params (EndPlayReason :type EEndPlayReason%20Enum :default "Destroyed"))
  ;[0.1198 0.05427 0] EndPlay: Clean up the actor.
  (|CallParentFunction EndPlayReason)
  (Networking|SwitchHasAuthority
    (:Remote
      (CallFunction|MilkshakeSplineHide) ; Destroy any debug spheres spawned using cheats.
    )))

(event Custom|MilkshakeSpawnEarthquake
  (params (EventName :type String))
  ;[0 0 0] Cheats
  ;;[0 0 0] Cheat ke * MilkshakeSpawnEarthquake (EventName)
  ;;
  ;; Spawns a single earthquake for the provided event name.
  (Networking|SwitchHasAuthority
    (:Authority
      (bind _earthquakeinfodatatable (Variables|Default|GetEarthquakeInfoDatatable)
        (:then
          (bind _returnvalue (Utilities|GetDataTableRow _earthquakeinfodatatable (Utilities|String|StringToName EventName)))
          (Game|SpawnActorfromClass "/Milkshake/Gameplay/BP_Milkshake_Earthquake.BP_Milkshake_Earthquake_C" (Math|Transform|MakeTransform (CallFunction|GetSplineLocationatAlpha (Math|Vector2D|BreakVector2D (.EarthquakeSplineMoveNormalizedLength_110_2E9091BE45B8E7FF70C161AE980C69A9 (Utilities|Struct|BreakMilkshakeEarthquakeInfoStruct _returnvalue))))) "AlwaysSpawn" "OverrideRootScale" 0 self _returnvalue 0 0 0 0))
        (:else)))))

(event Custom|MilkshakeSetCalendarRatio
  (params (Ratio :type Float%20%28double-precision%29))
  ;[0 0 0] Cheats
  ;;[0 0 0] Cheat ke * MilkshakeSetCalendarRatio (float)
  ;;
  ;; Update the force calendar time ratio.
  (Networking|SwitchHasAuthority
    (:Authority
      (Variables|Cheat|SetCheat_ForceCalendarTimeSpanRatio Ratio))))

(event Custom|MilkshakeStartEarthquakeEvent
  (params (EventName :type String))
  ;[0 0 0] Cheats
  ;;[0 0 0] Cheat ke * MilkshakeStartEarthquakeEvent (EventName)
  ;;
  ;; Spawns the entire sequence of earthquakes for that event entry.
  (Networking|SwitchHasAuthority
    (:Authority
      (bind _earthquakeinfodatatable (Variables|Default|GetEarthquakeInfoDatatable)
        (:then
          (bind _returnvalue (Utilities|GetDataTableRow _earthquakeinfodatatable (Utilities|String|StringToName EventName)))
          (bind (_enabled_104_9e9bd6984687100ae0fd3cb84b7e6a76 _calendareventratio_116_db14dea24c94389f1ded8fad08f78569) (Utilities|Struct|BreakMilkshakeEarthquakeInfoStruct _returnvalue))
          (bind (_calendareventratio_116_db14dea24c94389f1ded8fad08f78569_x _calendareventratio_116_db14dea24c94389f1ded8fad08f78569_y) (Math|Vector2D|BreakVector2D _calendareventratio_116_db14dea24c94389f1ded8fad08f78569))
          (CallFunction|MilkshakeSetCalendarRatio (Math|Random|RandomFloatinRange _calendareventratio_116_db14dea24c94389f1ded8fad08f78569_x _calendareventratio_116_db14dea24c94389f1ded8fad08f78569_y)))
        (:else)))))

(event Custom|MilkshakeStopAllEarthquakeEvents
  ;[0 0 0] Cheats
  ;;[0 0 0] Cheat ke * MilkshakeStopAllEarthquakeEvents (EventName)
  ;;
  ;; Reset the setup forced calendar time span ratio.
  (Networking|SwitchHasAuthority
    (:Authority
      (CallFunction|MilkshakeSetCalendarRatio -1.0))))

(event Custom|MilkshakeSplineShow
  (bind _cheat_sphereradius (Variables|Cheat|GetCheat_SphereRadius))
  (bind _returnvalue (* _cheat_sphereradius 2.0))
  ;[0 0 0] Cheats
  ;;[0 0 0] ke * MilkshakeSplineShow (Runs only for the player entering the cheat locally)
  ;;
  ;; Spawn debug spheres along the earthquake spline so that it can be debugged/tested.
  (Networking|CanExecuteCosmeticEvents
    (:True
      (if (not (Variables|Cheat|GetbCheat_SplineShowing))
        (Variables|Cheat|SetbCheat_SplineShowing true)
        (Variables|Default|GetSpline
          (:then
            (bind _returnvalue_2 (Utilities|Time|SetTimerbyEvent (EventDispatchers|CreateEvent :!event Cheat_EventNameTextCompFaceCamera) 0.1 true))
            (Variables|Cheat|SetCheat_TextCompFaceCameraTimerHandle _returnvalue_2)
            (Utilities|FlowControl|Sequence
              (:then_0
                ;;;[0.000895 0 0.09375] Spawn debug spheres along the spline.
                (for _index (range (+ (Math|Float|Truncate (Math|Float|SafeDivide (Spline|GetSplineLength (Variables|Default|GetSpline)) _returnvalue)) 1))
                  (Cheat|CheatSpawnSimpleInstancedSphere (Spline|GetLocationatDistanceAlongSpline (Variables|Default|GetSpline) (* _index _returnvalue) "World"))))
              (:then_1
                ;;;[0 0.03954 0.1927] Draw event names on top of the spline.
                (bind _earthquakeinfodatatable (Variables|Default|GetEarthquakeInfoDatatable)
                  (:then
                    (bind _outrownames (DataTable|GetDataTableRowNames _earthquakeinfodatatable))
                    (for _array_element _outrownames
                      (bind _returnvalue_3 (Utilities|GetDataTableRow _earthquakeinfodatatable _array_element))
                      (bind _returnvalue_4 (AddComponent|NODE_AddTextRenderComponent-0 :RelativeTransform_Location "0, 0, 0" :RelativeTransform_Rotation "0, 0, 0" :RelativeTransform_Scale "1.000000,1.000000,1.000000"))
                      (Utilities|IsValid _returnvalue_4
                        (:Is%20Valid
                          (bind (_enabled_104_9e9bd6984687100ae0fd3cb84b7e6a76 _calendareventratio_116_db14dea24c94389f1ded8fad08f78569 _earthquakesplinemovenormalizedlength_110_2e9091be45b8e7ff70c161ae980c69a9) (Utilities|Struct|BreakMilkshakeEarthquakeInfoStruct _returnvalue_3))
                          (bind (_calendareventratio_116_db14dea24c94389f1ded8fad08f78569_x _calendareventratio_116_db14dea24c94389f1ded8fad08f78569_y) (Math|Vector2D|BreakVector2D _calendareventratio_116_db14dea24c94389f1ded8fad08f78569))
                          (Rendering|Components|TextRender|SetText _returnvalue_4 (Utilities|Text|FormatText "INVTEXT(\"{RowName}\\r\\n{RatioStart} - {RatioEnd}\")" (Utilities|Text|ToText(Name) _array_element) _calendareventratio_116_db14dea24c94389f1ded8fad08f78569_x _calendareventratio_116_db14dea24c94389f1ded8fad08f78569_y))
                          (Utilities|Array|Add (Variables|Cheat|GetCheat_EventNameTextComponents) _returnvalue_4)
                          (Transformation|SetWorldLocation _returnvalue_4 (Utilities|Operators|Add (CallFunction|GetSplineLocationatAlpha (Math|Vector2D|BreakVector2D _earthquakesplinemovenormalizedlength_110_2e9091be45b8e7ff70c161ae980c69a9)) (Math|Vector|MakeVector 0.0 0.0 (Variables|Cheat|GetCheat_TextComponentLocationOffsetZ))) false false)))))
                  (:else)))))
          (:else))))))

(event Custom|MilkshakeSplineHide
  (bind _cheat_eventnametextcomponents (Variables|Cheat|GetCheat_EventNameTextComponents))
  ;[0 0 0] Cheats
  ;;[0 0 0] ke * MilkshakeSplineHide (Runs only for the player entering the cheat locally)
  ;;
  ;; Despawn the debug spheres spawned on the earthquake spline.
  (Networking|CanExecuteCosmeticEvents
    (:True
      (bind _cheat_instancedstaticmeshcomponent (Variables|Default|GetCheat_InstancedStaticMeshComponent)
        (:then
          (Components|InstancedStaticMesh|ClearInstances _cheat_instancedstaticmeshcomponent))
        (:else
          _))
      (Components|DestroyComponent _cheat_eventnametextcomponents)
      (Utilities|Array|Clear _cheat_eventnametextcomponents)
      (Variables|Cheat|SetbCheat_SplineShowing false)
      (Utilities|Time|ClearandInvalidateTimerbyHandle (Variables|Cheat|GetCheat_TextCompFaceCameraTimerHandle)))))

(event Custom|MilkshakeSetEarthquakeShow
  (params (bShow :type Boolean))
  ;[0 0 0] Cheats
  ;;[0 0 0] ke * MilkshakeSetEarthquakeShow (Runs only for the player entering the cheat locally)
  ;;
  ;; Draws a sphere to represent the earthquake happening in the world.
  (Networking|CanExecuteCosmeticEvents
    (:True
      (Variables|Cheat|SetbCheat_EarthquakeShow bShow))))

(fn Cheat_EventNameTextCompFaceCamera
  (bind _returnvalue (Fortnite|GetFirstLocalFortPlayerController true true))
  ;[0.05625 0.09679 0.125] Cheat_EventNameTextCompFaceCamera: Have the text components face the camera.
  (Networking|CanExecuteCosmeticEvents
    (:True
      (Utilities|IsValid _returnvalue
        (:Is%20Valid
          (bind _playercameramanager (Variables|/Script/Engine.PlayerController|GetPlayerCameraManager _returnvalue)
            (:then
              (for _array_element (Variables|Cheat|GetCheat_EventNameTextComponents)
                (Utilities|IsValid _array_element
                  (:Is%20Valid
                    (Transformation|SetWorldRotation _array_element (Math|Rotator|FindLookatRotation (Transformation|GetWorldLocation _array_element) (Camera|GetCameraLocation _playercameramanager)) false false)))))
            (:else)))))))

(fn Cheat_GetEarthquakeShow
  (return (Variables|Cheat|GetbCheat_EarthquakeShow)))

(fn Cheat_SpawnSimpleInstancedSphere
  (params (SpawnLocation :type Vector))
  ;[0.004774 0.01 0.1094] Cheat_SpawnSimpleInstancedSphere: Spawn mesh instances in the world on the earthquake spline.
  (Networking|CanExecuteCosmeticEvents
    (:True
      (Variables|Default|GetCheat_InstancedStaticMeshComponent
        (:then
          _)
        (:else
          (bind _returnvalue (AddComponent|NODE_AddInstancedStaticMeshComponent-0 :RelativeTransform_Location "0, 0, 0" :RelativeTransform_Rotation "0, 0, 0" :RelativeTransform_Scale "1.000000,1.000000,1.000000"))
          (Utilities|IsValid _returnvalue
            (:Is%20Valid
              (bind _output_get (Variables|Default|SetCheat_InstancedStaticMeshComponent _returnvalue))
              (Rendering|Material|SetVectorParameterValueonMaterials _output_get (LocalVariables|GetColorParamName) (Math|Conversions|ToVector(LinearColor) (LocalVariables|GetColor))) ; Update sphere color.
            ))))
      (bind _cheat_instancedstaticmeshcomponent (Variables|Default|GetCheat_InstancedStaticMeshComponent)
        (:then
          (Components|InstancedStaticMesh|AddInstance _cheat_instancedstaticmeshcomponent (Math|Transform|MakeTransform SpawnLocation) true))
        (:else)))))

(fn CreateEarthquakeComponent
  (params (EarthquakeEventName :type String%20%28by%20ref%29))
  ;[0.03494 0.0108 0.1927] CreateEarthquakeComponent: Creates the earthquake component to handle the earthquake for the input event.
  (Utilities|FlowControl|Sequence
    (:then_0
      (Networking|SwitchHasAuthority
        (:Authority
          (bind _returnvalue (Utilities|GetDataTableRow (Variables|Default|GetEarthquakeInfoDatatable) (Utilities|String|StringToName EarthquakeEventName))
            (:then
              ;;[0.5521 0 0.4212] Check if it is hotfixed off.
              (if (> (ScalableFloat|GetValueatLevel (Utilities|Struct|BreakMilkshakeEarthquakeInfoStruct _returnvalue)) 0)
                ;;[0 0.03727 0.07292] Check if earthquake is allowed in current playlist
                (bind (_ballow _playlistname) (CallFunction|DoesCurrentPlaylistAllowEarthquake (.RequiredPlaylistTags_113_33DFD1284D4FDA8A1D11D8804B982C77 (Utilities|Struct|BreakMilkshakeEarthquakeInfoStruct _returnvalue))))
                (if _ballow
                  ;;[0.2552 0.2392 0.1551] Check if there isn't an earthquake component for the event already.
                  (bind _earthquakecomponent (CallFunction|GetEarthquakeComponentforEvent EarthquakeEventName))
                  (bind _output_get (LocalVariables|SetExistingComponent _earthquakecomponent))
                  (Utilities|IsValid _output_get
                    (:Is%20Valid
                      ;;[1 0 0] Warning: Earthquake component for calendar event already exists.
                      (Fortnite|FortShippingLog self (Utilities|String|Append "CreateEarthquakeComponent: An earthquake component for earthquake event  " (Variables|GetEarthquakeEventName) " is already spawned and active. Existing component is: " (Utilities|GetDisplayName _output_get)) true))
                    (:Is%20Not%20Valid
                      (bind _returnvalue_2 (Game|AddComponentbyClass 0 "/Milkshake/Gameplay/BP_MilkshakeEarthquakeManagerComponent.BP_MilkshakeEarthquakeManagerComponent_C" EarthquakeEventName))
                      (Utilities|IsValid _returnvalue_2
                        (:Is%20Valid
                          (Fortnite|FortShippingLog self (Utilities|String|Append "CreateEarthquakeComponent: Sucessfully spawned component " (Utilities|GetDisplayName _returnvalue_2) " for earthquake event name " (Variables|GetEarthquakeEventName)))
                          (return true))
                        (:Is%20Not%20Valid
                          ;;[1 0 0] Warning: Failed to spawn component.
                          (Fortnite|FortShippingLog self (Utilities|String|Append "CreateEarthquakeComponent: Failed to spawn component for earthquake event name " (Variables|GetEarthquakeEventName)) true)))))
                  (else
                    ;;[0 0.093 0.1302] Earthquake not allowed for current loaded playlist.
                    (Fortnite|FortShippingLog self (Utilities|String|Append "CreateEarthquakeComponent: Failed, earthquake event " EarthquakeEventName " is NOT allowed for current loaded playlist: " _playlistname))))
                (else
                  ;;[0 0.5668 1] Earthquake event hotfixed off.
                  (Fortnite|FortShippingLog self (Utilities|String|Append "CreateEarthquakeComponent: Failed, earthquake event " EarthquakeEventName " is hotfixed off.")))))
            (:RowNotFound
              ;;[1 0 0] Warning: Earthquake event not found.
              (Fortnite|FortShippingLog self (Utilities|String|Append "CreateEarthquakeComponent: Earthquake event name  " EarthquakeEventName " was NOT found in the earthquake info datatable.") true))))))
    (:then_1
      (return false))))

(fn DoesCurrentPlaylistAllowEarthquake
  (params (RequiredPlaylistTags :type Gameplay%20Tag%20Container%20Structure%20%28by%20ref%29))
  (bind _returnvalue (Fortnite|Athena|GameState|GetGameStateAthena))
  ;[0.03454 0 0.1719] Returns if the loaded playlist has any of the required playlist tags.
  (Utilities|FlowControl|Sequence
    (:then_0
      (Utilities|IsValid _returnvalue
        (:Is%20Valid
          (bind _output_get (LocalVariables|SetCurrentGameState _returnvalue))
          (if (Utilities|GameplayTagContainerNotEmpty RequiredPlaylistTags)
            (return (GameplayTags|HasAnyTags (Class|FortGameStateAthena|GetAthenaPlaylistContextTags _output_get) RequiredPlaylistTags true) (Utilities|String|ToString(Name) (Playlist|GetCurrentPlaylistName _output_get)))
            (else
              (return true (Utilities|String|ToString(Name) (Playlist|GetCurrentPlaylistName _output_get))))))))
    (:then_1
      (return false ""))))

(fn Editor_DestroyAllDebugEarthquake
  ;[1 0.4428 0] Destroy all the debug radius actors spawned in the world.
  (bind _outactors (Actor|GetAllActorsOfClass "/Milkshake/Gameplay/BP_Milkshake_EarthquakeDebug.BP_Milkshake_EarthquakeDebug_C"))
  (for _array_element _outactors
    (Utilities|IsValid _array_element
      (:Is%20Valid
        (Utilities|Casting|CastToBP_Milkshake_Spline_Earthquake (Actor|GetAttachParentActor _array_element))
        (Actor|DestroyActor _array_element)))))

(fn Editor_DrawDebugEarthquake
  ;[0 0.06269 0.2448] Editor_DrawDebugEarthquakeRadius: Spawn debug actors in the world that represent the earthquake (radius, start/end locations) based on the values entered in EarthquakeInfoDatatable.
  ; Warning: These debug actors are ACTUALLY spawned into the level so it is important to clear them up using the Editor_DestroyAllDebugEarthquakeRadius before saving the level!
  ; Note: IF DebugEarthquakeDataTableRowNames is empty, all the earthquakes from the EarthquakeInfoDatatable is debugged.
  (if (Editor-Debug|IsDebugEnabled)
    (Editor-Debug|EditorDestroyAllDebugEarthquake) ; Destroy any existing radius debug actor before spawning new ones.
    (bind _earthquakeinfodatatable (Variables|Default|GetEarthquakeInfoDatatable)
      (:then
        (bind _outrownames (DataTable|GetDataTableRowNames _earthquakeinfodatatable))
        (for _array_element (select (> (Utilities|Array|Length (Variables|Editor-Debug|GetDebugEarthquakeDataTableRowNames)) 0) (Variables|Editor-Debug|GetDebugEarthquakeDataTableRowNames) _outrownames)
          (bind _returnvalue (Utilities|GetDataTableRow _earthquakeinfodatatable _array_element))
          (bind _returnvalue_2 (Game|SpawnActorfromClass "/Milkshake/Gameplay/BP_Milkshake_EarthquakeDebug.BP_Milkshake_EarthquakeDebug_C" (Math|Conversions|ToTransform(Vector) (CallFunction|GetSplineLocationatAlpha (Math|Vector2D|BreakVector2D (.EarthquakeSplineMoveNormalizedLength_110_2E9091BE45B8E7FF70C161AE980C69A9 (Utilities|Struct|BreakMilkshakeEarthquakeInfoStruct _returnvalue))))) "AlwaysSpawn" "OverrideRootScale" 0 (Utilities|Text|ToText(Name) _array_element) (CallFunction|GetSplineLocationatAlpha (.Y (Math|Vector2D|BreakVector2D (.EarthquakeSplineMoveNormalizedLength_110_2E9091BE45B8E7FF70C161AE980C69A9 (Utilities|Struct|BreakMilkshakeEarthquakeInfoStruct _returnvalue))))) (Math|Float|Max(Float) (Math|Vector2D|BreakVector2D (.EarthquakeRadius_98_1E491F714C8AF8BBEADEE384F2595AD4 (Utilities|Struct|BreakMilkshakeEarthquakeInfoStruct _returnvalue))) (.Y (Math|Vector2D|BreakVector2D (.EarthquakeRadius_98_1E491F714C8AF8BBEADEE384F2595AD4 (Utilities|Struct|BreakMilkshakeEarthquakeInfoStruct _returnvalue))))) (CallFunction|GetSplineLocationatAlpha (Math|Vector2D|BreakVector2D (.EarthquakeSplineMoveNormalizedLength_110_2E9091BE45B8E7FF70C161AE980C69A9 (Utilities|Struct|BreakMilkshakeEarthquakeInfoStruct _returnvalue)))) 0))
          (Utilities|IsValid _returnvalue_2
            (:Is%20Valid
              (Transformation|AttachActorToActor _returnvalue_2 self "None" "KeepWorld" "KeepWorld" "KeepWorld" false) ; Attach it to the earthquake so its easier to parse/understand\n when looking at the world outliner.
            ))))
      (:else))))

(fn GetCalendarEventStatus
  (bind _cheat_forcecalendartimespanratio (Variables|Cheat|GetCheat_ForceCalendarTimeSpanRatio))
  (bind (_calendarstatus_bactive_1_0d053231499d475d43be7789c3e9aa13 _calendarstatus_timespanratio_4_41a67f774e2fe74e26c7968c3358303b) (Utilities|Struct|BreakMilkshakeCalendarEventInfoStruct (Variables|GetCalendarStatus_bActive_1_0D053231499D475D43BE7789C3E9AA13)))
  ;[0 0.0407 0.08854] GetCalendarEventStatus: Takes debugging into consideration so that the feature can be tested by bypassing calendar checks.
  (Utilities|FlowControl|Sequence
    (:then_0
      (Networking|SwitchHasAuthority
        (:Authority
          (if (>= _cheat_forcecalendartimespanratio 0)
            (return true _cheat_forcecalendartimespanratio)
            (else
              (return _calendarstatus_bactive_1_0d053231499d475d43be7789c3e9aa13 _calendarstatus_timespanratio_4_41a67f774e2fe74e26c7968c3358303b))))))
    (:then_1
      (return false 0.0))))

(fn GetEarthquakeComponentForEvent
  (params (EventName :type String%20%28by%20ref%29))
  ;[0.151 0.03191 0.1437] GetEarthquakeComponentForEvent: Returns the component that is handling the earthquake with the input event name name.
  (Utilities|FlowControl|Sequence
    (:then_0
      (Networking|SwitchHasAuthority
        (:Authority
          (for _array_element (Actor|GetComponentsByClass "/Milkshake/Gameplay/BP_MilkshakeEarthquakeManagerComponent.BP_MilkshakeEarthquakeManagerComponent_C")
            (Utilities|IsValid _array_element
              (:Is%20Valid
                (if (Utilities|String|EqualExactly(String) (Class|BPMilkshakeEarthquakeManagerComponent|GetEventName _array_element) EventName)
                  (return _array_element))))))))
    (:then_1
      (return 0))))

(fn GetEarthquakeInfoForEvent
  (params (EventName :type String%20%28by%20ref%29))
  ;[0.2552 0.1031 0.03327] GetEarthquakeInfoForEvent: Returns the earthquake info struct from the data table with the input event name.
  (Utilities|FlowControl|Sequence
    (:then_0
      (bind _earthquakeinfodatatable (Variables|Default|GetEarthquakeInfoDatatable)
        (:then
          (bind _returnvalue (Utilities|GetDataTableRow _earthquakeinfodatatable (Utilities|String|StringToName EventName)))
          (return _returnvalue true))
        (:else)))
    (:then_1
      (return 0 false))))

(fn GetSplineLocationAtAlpha
  (params (Alpha :type Float%20%28double-precision%29))
  ;[0.02931 0 0.276] GetSplineLocationAtAlpha: Returns the world location along the spline at input alpha value if the entire spline length was normalized between 0.f and 1.0f.
  (bind _spline (Variables|Default|GetSpline)
    (:then
      (return (Spline|GetLocationatDistanceAlongSpline _spline (* (Spline|GetSplineLength _spline) (Math|Float|Clamp(Float) Alpha)) "World")))
    (:else
      ;;[1 0 0.145] If spline is invalid, return actor location.
      (return (Transformation|GetActorLocation)))))

(fn IsCalendarEventActive
  (params (EventName :type String))
  ;[0.0212 0.1094 0.06107] Returns if the input event name is active on the calendar.
  (Utilities|FlowControl|Sequence
    (:then_0
      (Networking|SwitchHasAuthority
        (:Authority
          (bind (_biseventactive _timespanratio) (CallFunction|GetCalendarEventStatus))
          (if _biseventactive
            (bind _earthquakeinfodatatable (Variables|Default|GetEarthquakeInfoDatatable)
              (:then
                (bind _returnvalue (Utilities|GetDataTableRow _earthquakeinfodatatable (Utilities|String|StringToName EventName)))
                (if (> (ScalableFloat|GetValueatLevel (Utilities|Struct|BreakMilkshakeEarthquakeInfoStruct _returnvalue)) 0)
                  (return (Math|Float|InRange(Float) _timespanratio (Math|Vector2D|BreakVector2D (.CalendarEventRatio_116_DB14DEA24C94389F1DED8FAD08F78569 (Utilities|Struct|BreakMilkshakeEarthquakeInfoStruct _returnvalue))) (.Y (Math|Vector2D|BreakVector2D (.CalendarEventRatio_116_DB14DEA24C94389F1DED8FAD08F78569 (Utilities|Struct|BreakMilkshakeEarthquakeInfoStruct _returnvalue))))))))
              (:else))))))
    (:then_1
      (return false))))

(fn IsDebugEnabled
  ;[0.1614 0.00213 0.1719] Returns true if bDebug is set to true and the game is running in editor.
  (return (and (Variables|Editor-Debug|GetbDebug) (Editor|InEditororPIE)) (and (and (Variables|Editor-Debug|GetbDebug) (Editor|InEditororPIE)) (Variables|Editor-Debug|GetbDebugDrawRuntimeOutlines))))

(fn UpdateCalendarEventStatus
  ;[0 0.03738 0.1302] Update the calendar event status responsible for the earthquakes on the spline.
  (Networking|SwitchHasAuthority
    (:Authority
      (bind _asfort_game_instance (Utilities|Casting|CastToFortGameInstance (Game|GetGameInstance)))
      (Utilities|IsValid (Class|FortGameInstance|GetEventManager _asfort_game_instance)
        (:Is%20Valid
          (bind (_biseventactive _timesincebegin _timeuntilend _timespanratio) (CalendarEvents|GetCalendarEventInformation (Variables|Default|GetCalendarEvent)))
          (Variables|SetCalendarStatus_bActive_1_0D053231499D475D43BE7789C3E9AA13 (Utilities|Struct|MakeMilkshakeCalendarEventInfoStruct _biseventactive _timespanratio)))) ; Make sure GetEventManger is valid before \nchecking calendar event info to prevent error spam.
    )))

(fn ConstructionScript)
